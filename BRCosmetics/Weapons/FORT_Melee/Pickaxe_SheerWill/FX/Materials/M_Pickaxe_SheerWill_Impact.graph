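# Translucent unlit, two-sided Niagara sprite and mesh particle material for the SheerWill pickaxe impact effect.
# Panning radial-UV nebula and noise textures build a tinted emissive, with scalar blends to a sprite-mask outline or simple glow.
# Opacity combines channel-selected spike mask, radial gradients and either streaming noise or Dissolve-driven height lerp.
# Particle alpha, view-angle and depth fades, plus optional bottom mask and camera fade, shape final opacity.
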Material /BRCosmetics/Weapons/FORT_Melee/Pickaxe_SheerWill/FX/Materials/M_Pickaxe_SheerWill_Impact
graph {
 properties {
  bAutomaticallySetUsageInEditor=false
  bUsedWithNiagaraMeshParticles=true
  bUsedWithNiagaraSprites=true
  blendMode=BLEND_Translucent
  shadingModel=MSM_Unlit
  twoSided=true
  usedShadingModels=Unlit
 }
 nodes [
  Constant() {r=2}
  n1 = ParticleColor()
  n2 = VectorParameter() {parameterName=OutLineColor, defaultValue=<rgba 1 0 0.0044 1>}
  n3 = TextureSampleParameter2D(UVs=Panner(Coordinate=TextureCoordinate())) {parameterName=Tex_Mask, texture=&/T_Pickaxe_SheerWill_CenterSpikes}
  n4 = Panner(
   Coordinate=Multiply(
    A=MaterialFunctionCall() {materialFunction=&VectorToRadialValue}.Radial Coordinates
    B=Constant2Vector() {g=1, r=3}
   )
   Time=NamedRerouteUsage()
  ) {speedY=-1.2}
  n5 = Multiply(B=Power(Base=TextureSample(UVs=@n4) {texture=/Game/Characters/Player/Male/Medium/Bodies/M_Med_Soldier_04/Skins/Galaxy/Textures/T_FN_Nebula_style2}.RGB) {constExponent=1.1}) {constA=1}
  n6 = ScalarParameter() {parameterName=Des}
  n7 = VectorParameter() {parameterName=BackGroundColor, defaultValue=<rgba 0.5 0.025 0.025 1>}
  n8 = MaterialFunctionCall() {materialFunction=&VectorToRadialValue}
  n9 = Constant2Vector() {g=0.5, r=3}
  n10 = Panner(Coordinate=Add(A=Multiply(A=@n8.Radial Coordinates, B=@n9), B=NamedRerouteUsage()), Time=NamedRerouteUsage()) {speedY=-0.7}
  n11 = Multiply(A=Desaturation(
    None=TextureSample(UVs=@n10) {texture=/Game/Characters/Player/Female/Medium/Bodies/F_MED_SuperNovaTaro/FX/Material/T_SuperNovaTaro_Skin_Nebula}.RGB
    Fraction=ScalarParameter() {parameterName=Des}
   )) {constB=2}
  n12 = Panner(
   Coordinate=Multiply(
    A=MaterialFunctionCall() {materialFunction=&VectorToRadialValue}.Radial Coordinates
    B=Constant2Vector() {g=1, r=3}
   )
   Time=NamedRerouteUsage()
  ) {speedY=-1}
  n13 = LinearInterpolate(Alpha=Power(Base=TextureSample(UVs=@n12) {texture=/Game/Accessories/FORT_Backpacks/Backpack_F_MED_IchorIncisor/FX/_old/T_CrystalShield}.R) {constExponent=5}) {constB=15}
  n14 = Add(
   A=Multiply(A=Desaturation(None=@n5, Fraction=@n6), B=VectorParameter() {parameterName=CenterColorNoise, defaultValue=<rgba 3 0.1406 0.1406 0>}.RGB)
   B=Add(A=Multiply(A=@n7.RGB, B=@n11), B=Multiply(A=@n13, B=@n7.RGB))
  )
  n15 = LinearInterpolate(A=Multiply(A=@n14) {constB=0.3}, B=@n14, Alpha=Power(Base=@n3.B))
  n16 = LinearInterpolate(Alpha=TextureSample(UVs=TextureCoordinate() {uTiling=1.3, vTiling=0.7}) {samplerType=SAMPLERTYPE_Grayscale, texture=/BRCosmetics/Weapons/FORT_Melee/Pickaxe_Cube_Ninja_Male/FX/T_CellNoise}.R) {constA=0.7, constB=2}
  n17 = TextureCoordinate()
  n18 = Constant() {r=0.5}
  n19 = ColorRamp!(None=OneMinus(None=SphereMask(A=@n17, B=@n18) {attenuationRadius=0.3, hardnessPercent=0})) {colorCurve="<self>.ColorCurve"}
  n20 = Add(
   A=LinearInterpolate(B=Multiply(A=@n3.G, B=@n2.RGB), Alpha=ScalarParameter() {parameterName=UseSpriteTexture})
   B=Multiply(A=Multiply(A=@n15, B=@n16), B=ComponentMask(None=@n19) {b=true, g=true, r=true})
  )
  n21 = SphereMask(A=TextureCoordinate(), B=Constant() {r=0.5}) {attenuationRadius=0.5, hardnessPercent=0}
  n22 = Power(Base=OneMinus(None=Power(Base=@n21) {constExponent=1}))
  MaterialFunctionCall(Emissive=LinearInterpolate(A=Multiply(A=@n1.RGB, B=@n20), B=Multiply(A=@n22, B=@n1.RGB) {constA=1}, Alpha=ScalarParameter() {parameterName=UseSimpleglow})) {materialFunction=&MF_FX_EmissiveAdjustments}
  MaterialFunctionCall() {materialFunction=&DebugTimeSine}
  n23 = ScalarParameter() {defaultValue=100, parameterName="Fade Length (S)"}
  n24 = ScalarParameter() {defaultValue=24, parameterName="Fade Offset (S)"}
  n25 = Multiply(A=MaterialFunctionCall(Radius=ScalarParameter() {defaultValue=0.5, parameterName=EdgeMask}, Density=Constant() {r=3}) {materialFunction=&RadialGradientExponential}.RadialGradientExponential)
  n26 = Multiply(A=MaterialFunctionCall(Radius=Constant() {r=0.5}, Density=Constant() {r=5}) {materialFunction=&RadialGradientExponential}.RadialGradientExponential)
  n27 = Power(
   Base=StaticComponentMaskParameter(None=@n3.RGB) {defaultR=true, parameterName=RGB Mask}
   Exp=ScalarParameter() {defaultValue=1, parameterName=MaskPlower}
  ) {constExponent=1}
  n28 = Multiply(
   A=MaterialFunctionCall() {materialFunction=&VectorToRadialValue}.Radial Coordinates
   B=Constant2Vector() {g=0.6, r=4}
  )
  n29 = NamedRerouteUsage()
  n30 = Multiply(
   A=MaterialFunctionCall() {materialFunction=&VectorToRadialValue}.Radial Coordinates
   B=Constant2Vector() {g=0.5, r=5}
  )
  n31 = NamedRerouteUsage()
  n32 = LinearInterpolate(
   A=TextureSample(UVs=Panner(Coordinate=@n28, Time=@n29) {speedY=-1}) {texture=/BRCosmetics/Weapons/FORT_Melee/Pickaxe_TameEagle/FX/T_TameEagle_Pickaxe_Dissolve_Fire}.R
   B=TextureSample(UVs=Panner(Coordinate=@n30, Time=@n31) {speedY=-0.7}) {texture=/Game/Global/Textures/Default/Noise/T_SoftCells_Negative}.R
  )
  n33 = SphereMask(A=TextureCoordinate(), B=Constant() {r=0.5}) {attenuationRadius=0.5, hardnessPercent=0}
  n34 = LinearInterpolate(A=ScalarParameter() {defaultValue=-0.2, parameterName=Centerpower}, Alpha=Power(Base=OneMinus(None=@n33)) {constExponent=1.3}) {constA=-0.2, constB=2}
  n35 = LinearInterpolate(
   A=ScalarParameter() {defaultValue=-0.8, parameterName=TailHardnessLerpMin}
   B=ScalarParameter() {defaultValue=2, parameterName=TailHardnessLerpMax}
   Alpha=Power(Base=Power(Base=@n32) {constExponent=1.3}, Exp=Saturate(None=@n34))
  ) {constA=-1, constB=2}
  n36 = Saturate(None=LinearInterpolate(Alpha=Saturate(None=@n35)) {constB=2})
  n37 = DynamicParameter() {defaultValue=<rgba 1 1 1 0>, paramNames [Param1,Param2,Param3,Dissolve]}
  n38 = TextureSample(UVs=TextureCoordinate() {uTiling=2, vTiling=2}) {samplerType=SAMPLERTYPE_Grayscale, texture=&LowResBlurredNoise}
  n39 = Saturate(None=Multiply(A=MaterialFunctionCall(Transition Phase=@n37.Dissolve, Height Texture=@n38.R) {materialFunction=&HeightLerp}.Results) {constB=5})
  n40 = ScalarParameter() {parameterName=UseSpriteTexture}
  n41 = Multiply(A=Multiply(A=Multiply(A=@n27, B=@n26), B=LinearInterpolate(A=@n36, B=@n39, Alpha=@n40)), B=Multiply(A=Saturate(None=@n25)) {constB=2}) {constB=2}
  n42 = Multiply(A=MaterialFunctionCall(
    Radius=ScalarParameter() {defaultValue=0.3, parameterName="Radius (S)"}
    Density=ScalarParameter() {defaultValue=2, parameterName="Density (S)"}
    Invert Density=StaticBool() {value=true}
   ) {materialFunction=&RadialGradientExponential}.RadialGradientExponential)
  n43 = LinearInterpolate(B=Power(Base=OneMinus(None=@n42)), Alpha=ScalarParameter() {parameterName=UseBottomMask}) {constA=1}
  n44 = Multiply(A=ParticleColor().A, B=Saturate(None=Multiply(A=@n41, B=@n43)))
  n45 = ScalarParameter() {defaultValue=2, parameterName="FadeSensitivity (optional) (S)"}
  n46 = Saturate(None=Multiply(A=MaterialFunctionCall("FadeSensitivity (optional)"=@n45) {materialFunction=&FadeBasedOnViewAngle}.Result) {constB=5})
  Multiply(
   A=DepthFade(Opacity=Multiply(A=@n44, B=@n46), FadeDistance=ScalarParameter() {parameterName=FadeDistance})
   B=LinearInterpolate(
    B=MaterialFunctionCall(Fade Length=@n23, Fade Offset=@n24) {materialFunction=&CameraDepthFade}.Result
    Alpha=ScalarParameter() {parameterName=UseCameraFade}
   ) {constA=1}
  )
  n47 = Panner(
   Coordinate=Multiply(
    A=MaterialFunctionCall() {materialFunction=&VectorToRadialValue}.Radial Coordinates
    B=Constant2Vector() {g=1, r=3}
   )
   Time=NamedRerouteUsage()
  ) {speedY=-0.5}
  NamedRerouteDeclaration(None=Multiply(A=TextureSample(UVs=@n47) {texture=/Game/Global/Textures/Default/Tileable_Patterns/T_Blowing_Smoke_Pattern2_Tile_DISP}.R) {constB=0.15}) {name=Disturb, nodeColor=<rgba 0.0034 1 0.7721 1>}
  NamedRerouteDeclaration(None=Multiply(A=Time(), B=ScalarParameter() {defaultValue=1, parameterName=Speed})) {name=Speed, nodeColor=<rgba 1 0.2812 0 1>}
 ]
}
