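# Translucent unlit trail material for the Sheer Will pickaxe, used on Niagara ribbons and mesh particles.
# Emissive layers panning nebula, noise and mask textures with outline, color ramp and optional simple glow, on a clock blending smooth and stepped time.
# Opacity combines tail and whole-trail dissolves, lead/tail edge masks and depth, angle and optional camera fades.
# Vertex offset sets width via SplineThicken or mesh-plane expansion plus optional sine wobble, driven by dynamic parameters.
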Material /BRCosmetics/Weapons/FORT_Melee/Pickaxe_SheerWill/FX/Materials/M_Pickaxe_SheerWill_Trail
graph {
 properties {
  bAutomaticallySetUsageInEditor=false
  bUsedWithNiagaraMeshParticles=true
  bUsedWithNiagaraRibbons=true
  blendMode=BLEND_Translucent
  shadingModel=MSM_Unlit
  twoSided=true
  usedShadingModels=Unlit
 }
 nodes [
  Constant()
  n1 = TextureCoordinate() {vTiling=0.9}
  n2 = TextureSample(UVs=Panner(Coordinate=Add(A=@n1) {constB=0}, Time=NamedRerouteUsage()) {speedX=-1.2}) {texture=/Game/Effects/Fort_Effects/Textures/Tile/T_Noise_02}
  Desaturation(None=@n2.RGB, Fraction=ScalarParameter() {parameterName=ColorDes})
  n3 = Time() {bIgnorePause=true, bOverride_Period=true, period=6000}
  Divide(A=Ceil(None=Multiply(A=@n3) {constB=6})) {constB=0.5}
  n4 = Multiply(
   A=TextureCoordinate()
   B=ComponentMask(None=VectorParameter() {parameterName=AllDissolveUV, defaultValue=<rgba 1.5 0.7 0 0>}.RGB) {g=true, r=true}
  )
  n5 = ComponentMask(None=VectorParameter() {parameterName=AllDissolveOffset, defaultValue=<rgba 0 0.4 0 1>}.RGB) {g=true, r=true}
  n6 = MaterialFunctionCall(
   Transition Phase=DynamicParameter() {defaultValue=<rgba 1 0 1 1>, paramNames [SplineWidth,TimeOffset,Dissolve,Appear]}.Dissolve
   Height Texture=TextureSample(UVs=Add(A=@n4, B=@n5)) {samplerType=SAMPLERTYPE_Grayscale, texture=&LowResBlurredNoise}.R
  ) {materialFunction=&HeightLerp}
  n7 = Saturate(None=LinearInterpolate(Alpha=@n6.Results) {constA=-1, constB=3})
  n8 = Multiply(A=VectorParameter() {parameterName=DissolveEdge, defaultValue=<rgba 1 0 0 0>}.RGB, B=Power(Base=Sine(None=@n7)) {constExponent=10})
  n9 = ScalarParameter() {parameterName=UseTheAllDissolve}
  n10 = ParticleColor()
  n11 = VectorParameter() {parameterName=MainAlphaUV, defaultValue=<rgba 0.7 1 0 0>}
  n12 = Panner(Coordinate=Multiply(A=TextureCoordinate(), B=ComponentMask(None=@n11.RGB) {g=true, r=true}), Time=NamedRerouteUsage()) {speedX=-0.6}
  n13 = TextureSampleParameter2D(UVs=@n12) {parameterName=Tex_Mask, texture=&/T_Pickaxe_SheerWill_Trail01}
  n14 = Multiply(A=VectorParameter() {parameterName=OutLineColor, defaultValue=<rgba 1 0 0.0529 1>}.RGB, B=@n13.G)
  n15 = ScalarParameter() {parameterName=AddedOutLine}
  n16 = VectorParameter() {parameterName=BackgroundColor, defaultValue=<rgba 0.5 0.025 0.025 1>}
  n17 = Panner(Coordinate=Add(A=TextureCoordinate() {vTiling=0.5}) {constB=0}, Time=NamedRerouteUsage()) {speedX=-0.7}
  n18 = Desaturation(
   None=Multiply(
    A=TextureSample(UVs=@n17) {texture=/Game/Characters/Player/Female/Medium/Bodies/F_MED_SuperNovaTaro/FX/Material/T_SuperNovaTaro_Skin_Nebula}.RGB
    B=ScalarParameter() {defaultValue=2, parameterName=BackColorNoiseIntensity}
   ) {constB=2}
   Fraction=ScalarParameter() {parameterName=ColorDes}
  )
  n19 = Multiply(
   A=TextureCoordinate()
   B=ComponentMask(None=VectorParameter() {group=HighLight, parameterName=HighLightNoiseUV, defaultValue=<rgba 2 1.5 0 0>}.RGB) {g=true, r=true}
  )
  n20 = TextureSampleParameter2D(UVs=Panner(Coordinate=Add(A=@n19) {constB=0}, Time=NamedRerouteUsage()) {speedX=-0.8}) {
   group=HighLight
   parameterName=HighLightNoise
   samplerType=SAMPLERTYPE_LinearGrayscale
   texture=/BRCosmetics/Weapons/FORT_Melee/Pickaxe_TigerRootHype/FX/T_Pickaxe_TigerRootHype_Crystal
  }
  n21 = LinearInterpolate(
   B=ScalarParameter() {defaultValue=2, parameterName=CrystalMax}
   Alpha=Power(Base=@n20.R, Exp=ScalarParameter() {defaultValue=3, group=HighLight, parameterName=HighLightNoisePower})
  ) {constB=15}
  n22 = TextureSample(UVs=@n12) {
   samplerType=SAMPLERTYPE_LinearGrayscale
   texture=&/T_Pickaxe_SheerWill_TrailThin
  }
  n23 = Power(Base=@n22.R)
  n24 = Add(
   A=Multiply(A=Multiply(A=@n23, B=@n2.R), B=VectorParameter() {parameterName=CenterColorNoise, defaultValue=<rgba 18 0.8438 0.8438 0>}.RGB)
   B=Add(A=Multiply(A=@n16.RGB, B=@n18), B=Multiply(A=@n21, B=@n16.RGB))
  )
  n25 = TextureCoordinate() {uTiling=1.3, vTiling=0.7}
  n26 = Multiply(
   A=LinearInterpolate(A=Multiply(A=@n24) {constB=0.3}, B=@n24, Alpha=Power(Base=@n13.B))
   B=LinearInterpolate(Alpha=TextureSample(UVs=@n25) {samplerType=SAMPLERTYPE_Grayscale, texture=/BRCosmetics/Weapons/FORT_Melee/Pickaxe_Cube_Ninja_Male/FX/T_CellNoise}.R) {constA=0.7, constB=2}
  )
  n27 = TextureCoordinate()
  n28 = ComponentMask(None=ColorRamp!(None=ComponentMask(None=@n27) {r=true}) {colorCurve="<self>.ColorCurve"}) {b=true, g=true, r=true}
  n29 = Multiply(A=@n10.RGB, B=Add(A=LinearInterpolate(B=@n14, Alpha=@n15), B=Multiply(A=@n26, B=@n28)))
  n30 = OneMinus(None=ComponentMask(None=TextureCoordinate()) {r=true})
  n31 = ScalarParameter() {parameterName=SimpleColorMaskpower}
  n32 = Multiply(
   A=LinearInterpolate(
    A=ScalarParameter() {parameterName=SimpleColorMaskMin}
    B=ScalarParameter() {defaultValue=1, parameterName=SimpleColorMaskMax}
    Alpha=Power(Base=@n30, Exp=@n31) {constExponent=3}
   ) {constA=0.1, constB=3}
   B=@n10.RGB
  )
  n33 = ScalarParameter() {parameterName=UseSimpleglow}
  MaterialFunctionCall(Emissive=Add(A=LinearInterpolate(B=@n8, Alpha=@n9), B=LinearInterpolate(A=@n29, B=@n32, Alpha=@n33))) {materialFunction=&MF_FX_EmissiveAdjustments}
  MaterialFunctionCall() {materialFunction=&DebugTimeSine}
  n34 = Panner(Coordinate=TextureCoordinate() {uTiling=2, vTiling=1.5}, Time=NamedRerouteUsage()) {speedX=-0.5}
  NamedRerouteDeclaration(None=Multiply(
    A=TextureSample(UVs=@n34) {texture=/Game/Effects/Niagara/MasterMaterials/SplineThickenTrail/Texture/T_Pickaxe_SoapPocket_NoiseGlow}.R
    B=ScalarParameter() {parameterName=DisturbIntensity}
   ) {constB=0.2}) {name=Disturb, nodeColor=<rgba 0.0034 1 0.7721 1>}
  n35 = Multiply(A=Time() {bIgnorePause=true, bOverride_Period=true, period=6000}, B=Constant() {r=12}) {constB=6}
  n36 = LinearInterpolate(
   A=Time()
   B=Divide(A=Ceil(None=@n35), B=Constant() {r=4}) {constB=3}
   Alpha=ScalarParameter() {defaultValue=0.1, parameterName=TimeStuckLerp}
  ) {constAlpha=0.3}
  n37 = DynamicParameter() {defaultValue=<rgba 1 0 1 1>, paramNames [SplineWidth,TimeOffset,Dissolve,Appear]}
  NamedRerouteDeclaration(None=Multiply(A=Add(A=@n36, B=@n37.TimeOffset), B=ScalarParameter() {defaultValue=1, parameterName=Speed})) {name=Speed, nodeColor=<rgba 1 0.2812 0 1>}
  ParticleRandom()
  ParticleRandom()
  n38 = MaterialFunctionCall(
   Fade Length=ScalarParameter() {defaultValue=100, parameterName="Fade Length (S)"}
   Fade Offset=ScalarParameter() {defaultValue=24, parameterName="Fade Offset (S)"}
  ) {materialFunction=&CameraDepthFade}
  n39 = ScalarParameter() {parameterName=UseCameraFade}
  n40 = TextureCoordinate()
  n41 = Add(
   A=OneMinus(None=ComponentMask(None=@n40) {r=true})
   B=LinearInterpolate(Alpha=DynamicParameter() {defaultValue=<rgba 1 0 1 1>, paramNames [SplineWidth,TimeOffset,Dissolve,Appear]}.Appear) {constA=-1}
  )
  n42 = LinearInterpolate(B=Saturate(None=Multiply(A=@n41) {constB=5}), Alpha=ScalarParameter() {parameterName=UseAppear}) {constA=1}
  n43 = MaterialFunctionCall("FadeSensitivity (optional)"=ScalarParameter() {parameterName=AngleFade}) {materialFunction=&FadeBasedOnViewAngle}
  n44 = LinearInterpolate(B=@n7, Alpha=ScalarParameter() {parameterName=UseTheAllDissolve}) {constA=1}
  n45 = ComponentMask(None=VectorParameter() {parameterName=TailDissolveUV, defaultValue=<rgba 1 1 0 0>}.RGB) {g=true, r=true}
  n46 = Multiply(A=TextureCoordinate() {uTiling=1.5, vTiling=1.1}, B=@n45)
  n47 = NamedRerouteUsage()
  n48 = Multiply(A=@n45, B=TextureCoordinate() {uTiling=1.5, vTiling=1.4})
  n49 = NamedRerouteUsage()
  n50 = LinearInterpolate(
   A=TextureSample(UVs=Panner(Coordinate=@n46, Time=@n47) {speedX=-1}) {texture=/BRCosmetics/Weapons/FORT_Melee/Pickaxe_TameEagle/FX/T_TameEagle_Pickaxe_Dissolve_Fire}.R
   B=TextureSample(UVs=Panner(Coordinate=@n48, Time=@n49) {speedX=-1}) {texture=/Game/Global/Textures/Default/Noise/T_SoftCells_Negative}.R
  )
  n51 = ComponentMask(None=TextureCoordinate()) {r=true}
  n52 = Saturate(None=LinearInterpolate(Alpha=Power(Base=@n51) {constExponent=1.3}) {constA=-0.2})
  n53 = ComponentMask(None=TextureCoordinate()) {g=true}
  n54 = LinearInterpolate(Alpha=OneMinus(None=Sine(None=@n53) {period=2})) {constA=1, constB=3}
  n55 = LinearInterpolate(
   A=ScalarParameter() {defaultValue=-1, parameterName=TailHardnessLerpMin}
   B=ScalarParameter() {defaultValue=2, parameterName=TailHardnessLerpMax}
   Alpha=Power(Base=Power(Base=@n50) {constExponent=1.3}, Exp=Multiply(A=@n52, B=@n54))
  ) {constA=-1, constB=2}
  n56 = Saturate(None=LinearInterpolate(Alpha=Saturate(None=@n55)) {constB=2})
  n57 = Power(
   Base=StaticComponentMaskParameter(None=@n13.RGB) {defaultR=true, parameterName=RGB Mask}
   Exp=ScalarParameter() {defaultValue=1, parameterName=MaskPlower}
  ) {constExponent=1}
  n58 = Saturate(None=Multiply(A=Multiply(A=@n57, B=@n56)) {constB=2})
  n59 = ComponentMask(None=TextureCoordinate()) {r=true}
  n60 = Multiply(A=Add(A=OneMinus(None=@n59)) {constB=-0.05}, B=ScalarParameter() {defaultValue=10, parameterName=TailEdge}) {constB=10}
  n61 = TextureCoordinate()
  n62 = Multiply(
   A=Add(A=ComponentMask(None=@n61) {r=true}, B=ScalarParameter() {parameterName=LeadMaskOffset}) {constB=0}
   B=ScalarParameter() {defaultValue=20, parameterName=LeadMaskEdge}
  ) {constB=20}
  n63 = Multiply(A=Multiply(A=ParticleColor().A, B=Multiply(A=@n58, B=@n44)), B=Multiply(A=Saturate(None=@n60), B=Saturate(None=@n62)))
  n64 = ScalarParameter() {defaultValue=1, parameterName=FadeDistance}
  n65 = ComponentMask(None=TextureCoordinate() {coordinateIndex=1}) {r=true}
  n66 = Saturate(None=Multiply(A=OneMinus(None=@n65)) {constB=2})
  n67 = ScalarParameter() {parameterName=UseSecondUV}
  n68 = Multiply(
   A=Multiply(A=DepthFade(Opacity=@n63, FadeDistance=@n64), B=LinearInterpolate(B=@n66, Alpha=@n67) {constA=1})
   B=Saturate(None=Multiply(A=@n43.Result) {constB=5})
  )
  Reroute(None=Multiply(A=Multiply(A=@n68, B=@n42), B=LinearInterpolate(B=@n38.Result, Alpha=@n39) {constA=1}))
  n69 = TextureCoordinate()
  n70 = LinearInterpolate(Alpha=ParticleRandom()) {constA=1.5, constB=2}
  n71 = Add(
   A=ComponentMask(None=Multiply(A=@n69, B=@n70)) {r=true}
   B=DynamicParameter() {defaultValue=<rgba 1 0 1 1>, paramNames [SplineWidth,TimeOffset,Dissolve,Appear]}.TimeOffset
  )
  n72 = Multiply(A=Time()) {constB=-0.2}
  n73 = Multiply(
   A=Sine(None=Add(A=@n71, B=@n72)) {period=0.8}
   B=Transform(None=VectorParameter() {parameterName=WPODirection, defaultValue=<rgba 3 0 0 0>}.RGB) {transformSourceType=TRANSFORMSOURCE_Local}
  )
  n74 = ScalarParameter() {parameterName=UseWPOAnimation}
  n75 = ScalarParameter() {parameterName=SplineWidthMin}
  n76 = ComponentMask(None=TextureCoordinate()) {r=true}
  n77 = ScalarParameter() {parameterName=SplineWidthLeadMaskOffset}
  n78 = Saturate(None=Multiply(A=Add(A=@n76, B=@n77) {constB=0}, B=ScalarParameter() {defaultValue=5, parameterName=SplineWidthLeadEdge}) {constB=5})
  n79 = OneMinus(None=ComponentMask(None=TextureCoordinate()) {r=true})
  n80 = LinearInterpolate(Alpha=Saturate(None=Multiply(A=@n79) {constB=2})) {constA=0.6}
  n81 = ScalarParameter() {parameterName=UseTailRangeScale}
  n82 = Multiply(
   A=Multiply(A=Constant() {r=30}, B=DynamicParameter() {defaultValue=<rgba 1 0 1 1>, paramNames [SplineWidth,TimeOffset,Dissolve,Appear]}.SplineWidth)
   B=Multiply(A=LinearInterpolate(A=@n75, Alpha=@n78), B=LinearInterpolate(B=@n80, Alpha=@n81) {constA=1})
  )
  n83 = MaterialFunctionCall(WidthBase=@n82, WidthTip=@n82, Expand U or V UV Channel=StaticBool()) {materialFunction=&SplineThicken}
  n84 = Subtract(A=ComponentMask(None=TextureCoordinate()) {g=true}) {constB=0.5}
  n85 = MaterialFunctionCall() {materialFunction=&TangentBasis}
  n86 = LinearInterpolate(Alpha=VertexColor().R) {constA=0.8}
  n87 = ScalarParameter() {parameterName=UseVertexColor}
  n88 = Multiply(
   A=Multiply(A=Multiply(A=@n84, B=@n85.Y V), B=LinearInterpolate(B=@n86, Alpha=@n87) {constA=1})
   B=DynamicParameter() {defaultValue=<rgba 1 0 1 1>, paramNames [SplineWidth,TimeOffset,Dissolve,Appear]}.SplineWidth
  )
  n89 = ScalarParameter() {parameterName=Use MeshPlane Width}
  Reroute(None=Add(A=LinearInterpolate(A=@n83.WorldPositionOffset, B=@n88, Alpha=@n89), B=LinearInterpolate(B=@n73, Alpha=@n74)) {constB=0})
  Round(None=@n6.Results)
  TextureCoordinate() {uTiling=5, vTiling=4}
 ]
}
