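# Idle Niagara effect for the Sheer Will pickaxe, layering glowing spline meshes, a bottom glow, curl-noise sparks and a sprite with a light.
# One partitioned two-particle emitter draws both side splines; separate left, right and duplicate center emitters are disabled.
# User.VariantColorIndex drives scratch dynamic inputs that pick Blot, Citadel or Froth colors and intensities.
# Spline alpha also varies between front end and gameplay, and by platform.
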NiagaraSystem /BRCosmetics/Weapons/FORT_Melee/Pickaxe_SheerWill/FX/NS_Pickaxe_SheerWill_Idle
dependencies {
 usedDataInterfaces [
  &NiagaraDataInterfaceCurve
  &NiagaraDataInterfacePlatformSet
  &NiagaraDataInterfaceArrayFloat
  &NiagaraDataInterfaceArrayFloat3
 ]
 usedDynamicInputs []
 usedModules [
  &self:SystemSpawnScript.SystemScriptSource.SystemScriptGraph.NiagaraNodeAssignment_0.SetVariables_#1
  &self:SystemSpawnScript.SystemScriptSource.SystemScriptGraph.NiagaraNodeAssignment_1.SetVariables_#2
  &self:SystemSpawnScript.SystemScriptSource.SystemScriptGraph.NiagaraNodeAssignment_2.SetVariables_#3
  &self:SystemSpawnScript.SystemScriptSource.SystemScriptGraph.NiagaraNodeAssignment_3.SetVariables_#4
  &self:SystemSpawnScript.SystemScriptSource.SystemScriptGraph.NiagaraNodeAssignment_4.SetVariables_#5
  &SystemState
 ]
}
emitters {
 BottomGlow {
  emitterUpdateScript [
   &EmitterState(Life Cycle Mode=::NewEnumerator1, Loop Behavior=::NewEnumerator1, Loop Duration Mode=::NewEnumerator1, UseLoopDelay=false)
   &SpawnBurst_Instantaneous(Use Spawn Probability=false, UseLoopCountLimit=false)
  ]
  particleSpawnScript [
   /SetVariables(Particles.SizeAll:NiagaraFloat=&Multiply_Float(A=0.15, B=1)) {isSetParameters=true, module=SetVariables_#6}
   &InitializeParticle(
    Color Mode=::NewEnumerator0
    Mesh Scale Mode=::NewEnumerator3
    Position Offset=&AddVectorToPosition(Position=<5 -10 0>, Vector=$System.FX offset)
    Sprite Size Mode=::NewEnumerator3
    UsePositionOffset=true
    Write Material Random=true
   )
   &InitialMeshOrientation(Mesh Orientation Mode=::NewEnumerator7, Rotation=<-0.25 0 0>, Use Rotation Vector=true)
   &DynamicMaterialParameters(
    Index 0 Param 1=&Multiply_Float(A=0.8, B=$Particles.SizeAll)
    Module.Index0Float0Write=true
    Module.Index0Float1Write=true
    Module.Index0Float2Write=true
    Module.Index0Float3Write=true
   )
  ]
  particleUpdateScript [
   &FN_Mesh_Attributes(
    Color=&MakeLinearColorFromVectorAndFloat(
     "Vector (RGB)"=&Multiply_Vector(
      A=&MakeVectorFromLinearColorRGB(Color=&self:ScratchDynamicInput_01(
        BlotColor:LinearColor=<rgba 1 0.0743 0.1514 0.1>
        CitadelColor:LinearColor=<rgba 0 0.5 1 0.1>
        FrothColor:LinearColor=<rgba 0 0 0 1>
        VariantColorIndex:NiagaraInt32=&MakeIntFromFloat(Float=$User.VariantColorIndex)
       ))
      B=&Multiply_VectorByFloat(Vector=<1 1 1>, Float=3)
     )
     "Float (Alpha)"=&Multiply_Float(
      A=1
      B=&self:ScratchDynamicInput_03(
       Blot:NiagaraFloat=1
       Citadel:NiagaraFloat=1
       Froth:NiagaraFloat=2
       VariantColorIndex:NiagaraInt32=&MakeIntFromFloat(Float=$User.VariantColorIndex)
      )
     )
    )
    Create Color Attribute=true
    Module Output=::NewEnumerator1
   ) {module=FN_Mesh_Attributes002}
   &ParticleState(Kill Particles When Lifetime Has Elapsed=false)
   &ScaleMeshSize(Scale Factor=&Multiply_VectorByFloat(
     Vector=&Multiply_VectorByFloat(
      Vector=<0.9 0.9 1.2>
      Float=&RandomRangeFloat(Minimum=1, Maximum=1.1, Evaluation Type=::NewEnumerator1)
     )
     Float=$Particles.SizeAll
    ))
   &CameraOffset()
   &ScaleColor(Scale RGB=&Multiply_VectorByFloat(Vector=<1 1 1>, Float=2), ScaleA=true, ScaleRGB=true)
  ]
  rendererClasses [&NiagaraMeshRendererProperties]
  renderers [{rendererClass=&NiagaraMeshRendererProperties, rendererIndex=0}]
 }
 CenterSpline {
  emitterUpdateScript [
   /SetVariables(Emitter.Des:NiagaraFloat=&self:ScratchDynamicInput_08(
     Blot:NiagaraFloat=0.5
     Citadel:NiagaraFloat=0.5
     Froth:NiagaraFloat=1
     VariantColorIndex:NiagaraInt32=&MakeIntFromFloat(Float=$User.VariantColorIndex)
    )) {isSetParameters=true}
   &EmitterState(Life Cycle Mode=::NewEnumerator1, Loop Behavior=::NewEnumerator1, Loop Duration Mode=::NewEnumerator1, UseLoopDelay=false)
   &SpawnBurst_Instantaneous(Use Spawn Probability=false, UseLoopCountLimit=false)
  ]
  particleSpawnScript [
   &InitializeParticle(
    Color Mode=::NewEnumerator0
    Mesh Scale Mode=::NewEnumerator3
    Mesh Uniform Scale=1.05
    Position Offset=&AddVectorToPosition(Position=<5 -8 0>, Vector=$System.FX offset)
    Sprite Size Mode=::NewEnumerator3
    UsePositionOffset=true
    Write Material Random=true
   )
   &InitialMeshOrientation(Mesh Orientation Mode=::NewEnumerator7, Rotation=<0 0 0.25>, Use Rotation Vector=true)
   &DynamicMaterialParameters(
    Index 0 Param 1=0.9
    Index 0 Param 2=0.4
    Module.Index0Float0Write=true
    Module.Index0Float1Write=true
    Module.Index0Float2Write=true
    Module.Index0Float3Write=true
   )
   &InitialMeshOrientation(Mesh Orientation Mode=::NewEnumerator7, Use Rotation Vector=true) {module=InitialMeshOrientation001}
  ]
  particleUpdateScript [
   &FN_Mesh_Attributes(
    Color=&MakeLinearColorFromVectorAndFloat(
     "Vector (RGB)"=&Multiply_Vector(
      A=&MakeVectorFromLinearColorRGB(Color=&self:ScratchDynamicInput_01(
        BlotColor:LinearColor=<rgba 1 0.0146 0 0.4>
        CitadelColor:LinearColor=<rgba 0 0.2 0.6 0.4>
        FrothColor:LinearColor=<rgba 0 0 0 1>
        VariantColorIndex:NiagaraInt32=&MakeIntFromFloat(Float=$User.VariantColorIndex)
       ))
      B=<1 1 1>
     )
     "Float (Alpha)"=&Multiply_Float(
      A=&Multiply_Float(
       A=&FN_FrontEndFloatModifier(
        Gameplay=0.3
        Front End=&self:ScratchDynamicInput_03(
         Blot:NiagaraFloat=0.2
         Citadel:NiagaraFloat=0.4
         Froth:NiagaraFloat=0.2
         VariantColorIndex:NiagaraInt32=&MakeIntFromFloat(Float=$User.VariantColorIndex)
        )
       )
       B=&self:ScratchDynamicInput_03(
        Blot:NiagaraFloat=1
        Citadel:NiagaraFloat=1
        Froth:NiagaraFloat=3
        VariantColorIndex:NiagaraInt32=&MakeIntFromFloat(Float=$User.VariantColorIndex)
       )
      )
      B=&FN_PickFloatDependOnPlatform(TargetPlatformActive=1, TargetPlatformNoActive=0.7)
     )
    )
    Create Color Attribute=true
    Module Output=::NewEnumerator1
   )
   &FN_Mesh_Attributes(
    Color=&MakeLinearColorFromVectorAndFloat(
     "Vector (RGB)"=&Multiply_Vector(
      A=&MakeVectorFromLinearColorRGB(Color=&self:ScratchDynamicInput_01(
        BlotColor:LinearColor=<rgba 0.7 0.054 0.0445 1>
        CitadelColor:LinearColor=<rgba 0 1.5 3 1>
        FrothColor:LinearColor=<rgba 0 0 0 1>
        VariantColorIndex:NiagaraInt32=&MakeIntFromFloat(Float=$User.VariantColorIndex)
       ))
      B=<1 1 1>
     )
     "Float (Alpha)"=&Multiply_Float(A=1, B=1)
    )
    Create Color Attribute=true
    Module Output=::NewEnumerator1
   ) {module=FN_Mesh_Attributes001}
   &ParticleState(Kill Particles When Lifetime Has Elapsed=false)
   &CameraOffset(Camera Offset Amount=2)
  ]
  rendererClasses [&NiagaraMeshRendererProperties]
  renderers [
   {rendererClass=&NiagaraMeshRendererProperties, rendererIndex=0}
   {rendererClass=&NiagaraMeshRendererProperties, rendererIndex=1}
   {rendererClass=&NiagaraMeshRendererProperties, rendererIndex=2}
  ]
 }
 CenterSpline001 {
  bEnabled=false
  emitterUpdateScript [
   &EmitterState(Life Cycle Mode=::NewEnumerator1, Loop Behavior=::NewEnumerator1, Loop Duration Mode=::NewEnumerator1, UseLoopDelay=false)
   /SetVariables(Emitter.BackGroundColor:LinearColor=&self:ScratchDynamicInput(
     BlotColor:LinearColor=<rgba 0.5 0.025 0.025 1>
     CitadelColor:LinearColor=<rgba 0 3.3333 10 1>
     FrothColor:LinearColor=<rgba 1 1 1 1>
     VariantColorIndex:NiagaraInt32=0
    )) {isSetParameters=true}
   /SetVariables(Emitter.CenterColorNoise:LinearColor=&self:ScratchDynamicInput(
     BlotColor:LinearColor=<rgba 18 0.8438 0.8438 0>
     CitadelColor:LinearColor=<rgba 0 33.3333 100 1>
     FrothColor:LinearColor=<rgba 1 1 1 1>
     VariantColorIndex:NiagaraInt32=0
    )) {isSetParameters=true}
   /SetVariables(Emitter.OutLineColor:LinearColor=&self:ScratchDynamicInput(
     BlotColor:LinearColor=<rgba 1 0 0.0529 1>
     CitadelColor:LinearColor=<rgba 0 0.5 1 1>
     FrothColor:LinearColor=<rgba 1 1 1 1>
     VariantColorIndex:NiagaraInt32=0
    )) {isSetParameters=true}
   /SetVariables(Emitter.DissolveEdge:LinearColor=&self:ScratchDynamicInput(
     BlotColor:LinearColor=<rgba 1 0 0 0>
     CitadelColor:LinearColor=<rgba 1 1 1 1>
     FrothColor:LinearColor=<rgba 1 1 1 1>
     VariantColorIndex:NiagaraInt32=0
    )) {isSetParameters=true}
   &SpawnBurst_Instantaneous(Use Spawn Probability=false, UseLoopCountLimit=false)
  ]
  particleSpawnScript [
   &InitializeParticle(
    Color Mode=::NewEnumerator0
    Mesh Scale Mode=::NewEnumerator3
    Mesh Uniform Scale=1.05
    Position Offset=&AddVectorToPosition(Position=<5 -8 0>, Vector=$System.FX offset)
    Sprite Size Mode=::NewEnumerator3
    UsePositionOffset=true
    Write Material Random=true
   )
   &FN_Mesh_Attributes(
    Color=&MakeLinearColorFromVectorAndFloat(
     "Vector (RGB)"=&Multiply_Vector(
      A=&MakeVectorFromLinearColorRGB(Color=&self:ScratchDynamicInput_01(
        BlotColor:LinearColor=<rgba 1 0.0146 0 0.4>
        CitadelColor:LinearColor=<rgba 0 0.2 0.6 0.4>
        FrothColor:LinearColor=<rgba 1 1 1 1>
        VariantColorIndex:NiagaraInt32=0
       ))
      B=<1 1 1>
     )
     "Float (Alpha)"=&Multiply_Float(A=0.3, B=1)
    )
    Create Color Attribute=true
    Lifetime=1
    Module Output=::NewEnumerator1
   )
   &FN_Mesh_Attributes(
    Color=&MakeLinearColorFromVectorAndFloat(
     "Vector (RGB)"=&Multiply_Vector(
      A=&MakeVectorFromLinearColorRGB(Color=&self:ScratchDynamicInput_01(
        BlotColor:LinearColor=<rgba 0.7 0.054 0.0445 1>
        CitadelColor:LinearColor=<rgba 0 1.5 3 1>
        FrothColor:LinearColor=<rgba 1 1 1 1>
        VariantColorIndex:NiagaraInt32=0
       ))
      B=<1 1 1>
     )
     "Float (Alpha)"=&Multiply_Float(A=1, B=1)
    )
    Create Color Attribute=true
    Lifetime=1
    Module Output=::NewEnumerator1
   ) {module=FN_Mesh_Attributes001}
   &InitialMeshOrientation(Mesh Orientation Mode=::NewEnumerator7, Rotation=<0 0 0.25>, Use Rotation Vector=true)
   &DynamicMaterialParameters(
    Index 0 Param 1=0.9
    Index 0 Param 2=0.4
    Module.Index0Float0Write=true
    Module.Index0Float1Write=true
    Module.Index0Float2Write=true
    Module.Index0Float3Write=true
   )
   &InitialMeshOrientation(Mesh Orientation Mode=::NewEnumerator7, Use Rotation Vector=true) {module=InitialMeshOrientation001}
  ]
  particleUpdateScript [
   &ParticleState(Kill Particles When Lifetime Has Elapsed=false)
   &CameraOffset(Camera Offset Amount=2)
  ]
  rendererClasses [&NiagaraMeshRendererProperties]
  renderers [
   {rendererClass=&NiagaraMeshRendererProperties, rendererIndex=0}
   {rendererClass=&NiagaraMeshRendererProperties, rendererIndex=1}
   {rendererClass=&NiagaraMeshRendererProperties, rendererIndex=2}
  ]
 }
 Dragonsparks {
  emitterUpdateScript [
   &EmitterState(Inactive Response=::NewEnumerator1, Life Cycle Mode=::NewEnumerator1, Loop Duration=4.6, UseLoopDelay=false)
   &SpawnRate(SpawnRate=10, Use Spawn Probability=false)
  ]
  particleSpawnScript [
   &InitializeParticle(
    Color=&self:ScratchDynamicInput_01(
     BlotColor:LinearColor=<rgba 50 0 4.1667 1>
     CitadelColor:LinearColor=<rgba 0 25 50 1>
     FrothColor:LinearColor=<rgba 0 0 0 100>
     VariantColorIndex:NiagaraInt32=&MakeIntFromFloat(Float=$User.VariantColorIndex)
    )
    Lifetime Max=0.9
    Lifetime Min=0.7
    Lifetime Mode=::NewEnumerator1
    Ribbon Width=40
    Ribbon Width Mode=::NewEnumerator1
    Sprite Size Max=<2 6>
    Sprite Size Min=<1 4>
    Sprite Size Mode=::NewEnumerator2
    UsePositionOffset=false
    Write Material Random=true
   )
   &ShapeLocation(MaskSpawnGroup=false, Offset=<4 10 0>, Offset Mode=::NewEnumerator0, Override Seed=false, Sphere Radius=7, UseOwnerScale=true)
   &AddVelocity(
    Module.Use Falloff Distance=false
    Origin Offset=<0 -30 0>
    Override Seed=false
    Velocity Mode=::NewEnumerator1
    Velocity Speed=&RandomRangeFloat(Minimum=50, Maximum=70)
   )
   /SetVariables(Particles.SizeAll=1) {isSetParameters=true}
  ]
  particleUpdateScript [
   &ParticleState()
   &ScaleSpriteSize(Scale Factor=&Vector2DFromFloat(Value=&FloatFromCurve(
      FloatCurve=DataInterface() {propertyValues {Curve {keys [
          <at 0 0 RCIM_Cubic>
          {interpMode=RCIM_Cubic, leaveTangent=-3.0296, tangentMode=RCTM_Break, time=0.2, value=1.2}
          {arriveTangent=-1.8107, interpMode=RCIM_Cubic, leaveTangent=-1.8107, tangentMode=RCTM_User, time=1, value=0.2}
         ]}}}
      Scale Curve=&Multiply_Float(A=0.4, B=$Particles.SizeAll)
     ))) {module=ScaleSpriteSize001}
   &SpriteRotationRate(Rotation Rate=&RandomRangeFloat(Minimum=-70, Maximum=70))
   /SetVariables(Particles.Glow03_Color=&MakeLinearColorFromVectorAndFloat(
     "Vector (RGB)"=&MakeVectorFromLinearColorRGB(Color=&MultiplyLinearColor(
       Linear Color=&self:ScratchDynamicInput_01(
        BlotColor:LinearColor=<rgba 1 0.0112 0.0112 1>
        CitadelColor:LinearColor=<rgba 0.0112 0.5056 1 1>
        FrothColor:LinearColor=<rgba 0 0 0 1>
        VariantColorIndex:NiagaraInt32=&MakeIntFromFloat(Float=$User.VariantColorIndex)
       )
       Scale Factor=&MakeVector4_FromFloat(Float=1)
      ))
     "Float (Alpha)"=&Multiply_Float(
      A=0.1
      B=&self:ScratchDynamicInput_03(Blot:NiagaraFloat=1, Citadel:NiagaraFloat=1, Froth:NiagaraFloat=2, VariantColorIndex:NiagaraInt32=0)
     )
    )) {isSetParameters=true}
   /SetVariables(Particles.Glow03_Size=&Multiply_Vector2D_ByFloat(
     Float=&FloatFromCurve(
      FloatCurve=DataInterface() {propertyValues {Curve {keys [
          <at 0 0 RCIM_Cubic>
          <at 0.2 1.3 RCIM_Cubic>
          {arriveTangent=-0.1423, interpMode=RCIM_Cubic, leaveTangent=-0.1423, tangentMode=RCTM_User, time=1, value=0}
         ]}}}
      Scale Curve=&Multiply_Float(A=6, B=$Particles.SizeAll)
     )
     Vector2D=<1 2>
    )) {isSetParameters=true}
   &DynamicMaterialParameters(
    Index 0 Param 1=&MakeCustomFloatFromBool(Bool=&RandomBool(), True Float=0, False Float=0.75)
    Module.Index0Float0Write=true
    Module.Index0Float1Write=true
    Module.Index0Float2Write=true
    Module.Index0Float3Write=true
   )
   &Drag(Drag=&RandomRangeFloat(Minimum=1, Maximum=1.5), Use Linear Drag=true, Use Rotational Drag=true)
   &CurlNoiseForce(Bias Noise=false, Noise Strength=50, Pan Noise=true, Pan Noise Field=<0 1 0>)
   &SolveForcesAndVelocity(Clamp Velocity=false, Limit Acceleration=false)
   &CameraOffset(Camera Offset Amount=$System.CameraOffsetAll) {enabled=false}
   &CameraOffset(Camera Offset Amount=-10) {module=CameraOffset001}
   &ScaleColor(
    Scale Alpha=&self:ScratchDynamicInput_03(Blot:NiagaraFloat=1, Citadel:NiagaraFloat=1, Froth:NiagaraFloat=100, VariantColorIndex:NiagaraInt32=0)
    ScaleA=true
    ScaleRGB=true
   ) {enabled=false}
  ]
  rendererClasses [&NiagaraSpriteRendererProperties]
  renderers [
   {rendererClass=&NiagaraSpriteRendererProperties}
   {rendererClass=&NiagaraSpriteRendererProperties, rendererIndex=1}
  ]
 }
 LeftSpline {
  bEnabled=false
  emitterUpdateScript [
   &EmitterState(Life Cycle Mode=::NewEnumerator1, Loop Behavior=::NewEnumerator1, Loop Duration Mode=::NewEnumerator1, UseLoopDelay=false)
   &SpawnBurst_Instantaneous(Use Spawn Probability=false, UseLoopCountLimit=false)
  ]
  particleSpawnScript [
   /SetVariables(Particles.SizeAll=0.7) {isSetParameters=true, module=SetVariables_#6}
   &InitializeParticle(
    Color Mode=::NewEnumerator0
    Mesh Scale Mode=::NewEnumerator3
    Position Offset=&AddVectorToPosition(Position=<5 -8 0>, Vector=$System.FX offset)
    Sprite Size Mode=::NewEnumerator3
    UsePositionOffset=true
    Write Material Random=true
   )
   &FN_Mesh_Attributes(
    Color=&MakeLinearColorFromVectorAndFloat(
     "Vector (RGB)"=&Multiply_Vector(A=&MakeVectorFromLinearColorRGB(Color=<rgba 1 0.0146 0 0.4>), B=<1 1 1>)
     "Float (Alpha)"=&Multiply_Float(A=0.3, B=1)
    )
    Create Color Attribute=true
    Lifetime=1
    Module Output=::NewEnumerator1
   )
   &FN_Mesh_Attributes(
    Color=&MakeLinearColorFromVectorAndFloat(
     "Vector (RGB)"=&Multiply_Vector(A=&MakeVectorFromLinearColorRGB(Color=<rgba 0.3 0.0232 0.0191 1>), B=<1 1 1>)
     "Float (Alpha)"=&Multiply_Float(A=1, B=1)
    )
    Create Color Attribute=true
    Lifetime=1
    Module Output=::NewEnumerator1
   ) {module=FN_Mesh_Attributes001}
   &InitialMeshOrientation(Mesh Orientation Mode=::NewEnumerator7, Rotation=<0 0 0.25>, Use Rotation Vector=true)
   &DynamicMaterialParameters(
    Index 0 Param 1=&Multiply_Float(A=1, B=$Particles.SizeAll)
    Module.Index0Float0Write=true
    Module.Index0Float1Write=true
    Module.Index0Float2Write=true
    Module.Index0Float3Write=true
   )
   &InitialMeshOrientation(Mesh Orientation Mode=::NewEnumerator7, Rotation=<0.05 0 0>, Use Rotation Vector=true) {module=InitialMeshOrientation001}
  ]
  particleUpdateScript [
   &ParticleState(Kill Particles When Lifetime Has Elapsed=false)
   &ScaleMeshSize(Scale Factor=&Multiply_VectorByFloat(Vector=<1 1 1>, Float=$Particles.SizeAll))
   &CameraOffset()
  ]
  rendererClasses [&NiagaraMeshRendererProperties]
  renderers [
   {rendererClass=&NiagaraMeshRendererProperties, rendererIndex=0}
   {rendererClass=&NiagaraMeshRendererProperties, rendererIndex=1}
   {rendererClass=&NiagaraMeshRendererProperties, rendererIndex=2}
  ]
 }
 LeftSpline001 {
  emitterUpdateScript [
   /SetVariables(Emitter.Des:NiagaraFloat=&self:ScratchDynamicInput_08(
     Blot:NiagaraFloat=0.5
     Citadel:NiagaraFloat=0.5
     Froth:NiagaraFloat=1
     VariantColorIndex:NiagaraInt32=&MakeIntFromFloat(Float=$User.VariantColorIndex)
    )) {isSetParameters=true}
   &EmitterState(Life Cycle Mode=::NewEnumerator1, Loop Behavior=::NewEnumerator1, Loop Duration Mode=::NewEnumerator1, UseLoopDelay=false)
   &SpawnBurst_Instantaneous(Spawn Count=2, Use Spawn Probability=false, UseLoopCountLimit=false)
  ]
  particleSpawnScript [
   &PartitionParticles()
   /SetVariables(Particles.SizeAll:NiagaraFloat=&SelectFloatFromArray(Direct Array Index=$Particles.PartitionParticles.Partition, Float Selection Array=DataInterface() {propertyValues {FloatData [0.7,0.8]}})) {isSetParameters=true, module=SetVariables_#6}
   &InitializeParticle(
    Color Mode=::NewEnumerator0
    Mesh Scale Mode=::NewEnumerator3
    Position Offset=&AddVectorToPosition(Position=<5 -8 0>, Vector=$System.FX offset)
    Sprite Size Mode=::NewEnumerator3
    UsePositionOffset=true
    Write Material Random=true
   )
   &InitialMeshOrientation(Mesh Orientation Mode=::NewEnumerator7, Rotation=<0 0 0.25>, Use Rotation Vector=true)
   &DynamicMaterialParameters(
    Index 0 Param 1=&Multiply_Float(
     A=&SelectFloatFromArray(Direct Array Index=$Particles.PartitionParticles.Partition, Float Selection Array=DataInterface() {propertyValues {FloatData [1,0.7]}})
     B=$Particles.SizeAll
    )
    Module.Index0Float0Write=true
    Module.Index0Float1Write=true
    Module.Index0Float2Write=true
    Module.Index0Float3Write=true
   )
   &InitialMeshOrientation(
    Mesh Orientation Mode=::NewEnumerator7
    Rotation=&SelectVectorFromArray(
     Direct Array Index=$Particles.PartitionParticles.Partition
     Vector Selection Array=DataInterface() {propertyValues {FloatData [<0.05 0 0>, <-0.06 0 0>]}}
    )
    Use Rotation Vector=true
   ) {module=InitialMeshOrientation001}
  ]
  particleUpdateScript [
   &FN_Mesh_Attributes(
    Color=&MakeLinearColorFromVectorAndFloat(
     "Vector (RGB)"=&Multiply_Vector(
      A=&MakeVectorFromLinearColorRGB(Color=&self:ScratchDynamicInput_01(
        BlotColor:LinearColor=<rgba 1 0.0146 0 0.4>
        CitadelColor:LinearColor=<rgba 0 0.2 0.6 0.4>
        FrothColor:LinearColor=<rgba 0 0 0 1>
        VariantColorIndex:NiagaraInt32=&MakeIntFromFloat(Float=$User.VariantColorIndex)
       ))
      B=<1 1 1>
     )
     "Float (Alpha)"=&Multiply_Float(
      A=&Multiply_Float(
       A=&FN_FrontEndFloatModifier(
        Gameplay=0.3
        Front End=&self:ScratchDynamicInput_03(
         Blot:NiagaraFloat=0.2
         Citadel:NiagaraFloat=0.4
         Froth:NiagaraFloat=0.2
         VariantColorIndex:NiagaraInt32=&MakeIntFromFloat(Float=$User.VariantColorIndex)
        )
       )
       B=&self:ScratchDynamicInput_03(
        Blot:NiagaraFloat=1
        Citadel:NiagaraFloat=1
        Froth:NiagaraFloat=3
        VariantColorIndex:NiagaraInt32=&MakeIntFromFloat(Float=$User.VariantColorIndex)
       )
      )
      B=&FN_PickFloatDependOnPlatform(TargetPlatformActive=1, TargetPlatformNoActive=0.7)
     )
    )
    Create Color Attribute=true
    Module Output=::NewEnumerator1
   )
   &FN_Mesh_Attributes(
    Color=&MakeLinearColorFromVectorAndFloat(
     "Vector (RGB)"=&Multiply_Vector(
      A=&MakeVectorFromLinearColorRGB(Color=&self:ScratchDynamicInput_01(
        BlotColor:LinearColor=<rgba 0.7 0.054 0.0445 1>
        CitadelColor:LinearColor=<rgba 0 1.5 3 1>
        FrothColor:LinearColor=<rgba 0 0 0 1>
        VariantColorIndex:NiagaraInt32=&MakeIntFromFloat(Float=$User.VariantColorIndex)
       ))
      B=<1 1 1>
     )
     "Float (Alpha)"=&Multiply_Float(A=1, B=1)
    )
    Create Color Attribute=true
    Module Output=::NewEnumerator1
   ) {module=FN_Mesh_Attributes001}
   &ParticleState(Kill Particles When Lifetime Has Elapsed=false)
   &ScaleMeshSize(Scale Factor=&Multiply_VectorByFloat(Vector=<1 1 1>, Float=$Particles.SizeAll))
   &CameraOffset()
   &ScaleColor(
    Scale Alpha=&self:ScratchDynamicInput_03(
     Blot:NiagaraFloat=1
     Citadel:NiagaraFloat=1
     Froth:NiagaraFloat=2
     VariantColorIndex:NiagaraInt32=&MakeIntFromFloat(Float=$User.VariantColorIndex)
    )
    ScaleA=true
    ScaleRGB=true
   )
  ]
  rendererClasses [&NiagaraMeshRendererProperties]
  renderers [
   {rendererClass=&NiagaraMeshRendererProperties, rendererIndex=0}
   {rendererClass=&NiagaraMeshRendererProperties, rendererIndex=1}
   {rendererClass=&NiagaraMeshRendererProperties, rendererIndex=2}
  ]
 }
 ParticleDecal {
  emitterUpdateScript [
   &EmitterState(Life Cycle Mode=::NewEnumerator1, Loop Behavior=::NewEnumerator1, Loop Duration Mode=::NewEnumerator1, UseLoopDelay=false)
   &SpawnBurst_Instantaneous(Use Spawn Probability=false, UseLoopCountLimit=false)
  ]
  particleSpawnScript [
   &InitializeParticle(
    Color=&self:ScratchDynamicInput_01(
     BlotColor:LinearColor=<rgba 5 0.5444 0.4489 1>
     CitadelColor:LinearColor=<rgba 0.4489 2.7245 5 1>
     FrothColor:LinearColor=<rgba 1 1 1 1>
     VariantColorIndex:NiagaraInt32=&MakeIntFromFloat(Float=$User.VariantColorIndex)
    )
    Sprite Size Mode=::NewEnumerator3
    Uniform Sprite Size=30
    UsePositionOffset=false
    Write Material Random=true
   )
   &CameraOffset(Camera Offset Amount=20)
  ]
  particleUpdateScript [
   &ParticleState(Kill Particles When Lifetime Has Elapsed=false)
   &Light_Attributes(
    Color=1
    Color Mode=0
    Falloff Exponent=2
    Light Color:LinearColor=&self:ScratchDynamicInput_01(
     BlotColor:LinearColor=<rgba 10 1.0888 0.8978 1>
     CitadelColor:LinearColor=<rgba 0 6.6667 20 1>
     FrothColor:LinearColor=<rgba 0 0 0 1>
     VariantColorIndex:NiagaraInt32=&MakeIntFromFloat(Float=$User.VariantColorIndex)
    )
    Light Exponent=3
    Light Radius=70
    Radius=2
   )
  ]
  rendererClasses [&NiagaraSpriteRendererProperties, &NiagaraLightRendererProperties]
  renderers [
   {rendererClass=&NiagaraSpriteRendererProperties}
   {rendererClass=&NiagaraLightRendererProperties, rendererIndex=1}
  ]
 }
 RightSpline {
  bEnabled=false
  emitterUpdateScript [
   &EmitterState(Life Cycle Mode=::NewEnumerator1, Loop Behavior=::NewEnumerator1, Loop Duration Mode=::NewEnumerator1, UseLoopDelay=false)
   &SpawnBurst_Instantaneous(Use Spawn Probability=false, UseLoopCountLimit=false)
  ]
  particleSpawnScript [
   /SetVariables(Particles.SizeAll=0.8) {isSetParameters=true, module=SetVariables_#6}
   &InitializeParticle(
    Color Mode=::NewEnumerator0
    Mesh Scale Mode=::NewEnumerator3
    Position Offset=&AddVectorToPosition(Position=<5 -8 0>, Vector=$System.FX offset)
    Sprite Size Mode=::NewEnumerator3
    UsePositionOffset=true
    Write Material Random=true
   )
   &FN_Mesh_Attributes(
    Color=&MakeLinearColorFromVectorAndFloat(
     "Vector (RGB)"=&Multiply_Vector(A=&MakeVectorFromLinearColorRGB(Color=<rgba 1 0.0146 0 0.4>), B=<1 1 1>)
     "Float (Alpha)"=&Multiply_Float(A=0.3, B=1)
    )
    Create Color Attribute=true
    Lifetime=1
    Module Output=::NewEnumerator1
   )
   &FN_Mesh_Attributes(
    Color=&MakeLinearColorFromVectorAndFloat(
     "Vector (RGB)"=&Multiply_Vector(A=&MakeVectorFromLinearColorRGB(Color=<rgba 0.3 0.0232 0.0191 1>), B=<1 1 1>)
     "Float (Alpha)"=&Multiply_Float(A=1, B=1)
    )
    Create Color Attribute=true
    Lifetime=1
    Module Output=::NewEnumerator1
   ) {module=FN_Mesh_Attributes001}
   &InitialMeshOrientation(Mesh Orientation Mode=::NewEnumerator7, Rotation=<0 0 0.25>, Use Rotation Vector=true)
   &DynamicMaterialParameters(
    Index 0 Param 1=&Multiply_Float(A=0.7, B=$Particles.SizeAll)
    Module.Index0Float0Write=true
    Module.Index0Float1Write=true
    Module.Index0Float2Write=true
    Module.Index0Float3Write=true
   )
   &InitialMeshOrientation(Mesh Orientation Mode=::NewEnumerator7, Rotation=<-0.06 0 0>, Use Rotation Vector=true) {module=InitialMeshOrientation001}
  ]
  particleUpdateScript [
   &ParticleState(Kill Particles When Lifetime Has Elapsed=false)
   &ScaleMeshSize(Scale Factor=&Multiply_VectorByFloat(Vector=<1 1 1>, Float=$Particles.SizeAll))
   &CameraOffset()
  ]
  rendererClasses [&NiagaraMeshRendererProperties]
  renderers [
   {rendererClass=&NiagaraMeshRendererProperties, rendererIndex=0}
   {rendererClass=&NiagaraMeshRendererProperties, rendererIndex=1}
   {rendererClass=&NiagaraMeshRendererProperties, rendererIndex=2}
  ]
 }
 RightSpline001 {
  bEnabled=false
  emitterUpdateScript [
   &EmitterState(Life Cycle Mode=::NewEnumerator1, Loop Behavior=::NewEnumerator1, Loop Duration Mode=::NewEnumerator1, UseLoopDelay=false)
   &SpawnBurst_Instantaneous(Use Spawn Probability=false, UseLoopCountLimit=false)
  ]
  particleSpawnScript [
   /SetVariables(Particles.SizeAll=0.8) {isSetParameters=true, module=SetVariables_#6}
   &InitializeParticle(
    Color Mode=::NewEnumerator0
    Mesh Scale Mode=::NewEnumerator3
    Position Offset=&AddVectorToPosition(Position=<5 -8 0>, Vector=$System.FX offset)
    Sprite Size Mode=::NewEnumerator3
    UsePositionOffset=true
    Write Material Random=true
   )
   &FN_Mesh_Attributes(
    Color=&MakeLinearColorFromVectorAndFloat(
     "Vector (RGB)"=&Multiply_Vector(A=&MakeVectorFromLinearColorRGB(Color=<rgba 1 0.0146 0 0.4>), B=<1 1 1>)
     "Float (Alpha)"=&Multiply_Float(A=0.3, B=1)
    )
    Create Color Attribute=true
    Lifetime=1
    Module Output=::NewEnumerator1
   )
   &FN_Mesh_Attributes(
    Color=&MakeLinearColorFromVectorAndFloat(
     "Vector (RGB)"=&Multiply_Vector(A=&MakeVectorFromLinearColorRGB(Color=<rgba 0.3 0.0232 0.0191 1>), B=<1 1 1>)
     "Float (Alpha)"=&Multiply_Float(A=1, B=1)
    )
    Create Color Attribute=true
    Lifetime=1
    Module Output=::NewEnumerator1
   ) {module=FN_Mesh_Attributes001}
   &InitialMeshOrientation(Mesh Orientation Mode=::NewEnumerator7, Rotation=<0 0 0.25>, Use Rotation Vector=true)
   &DynamicMaterialParameters(
    Index 0 Param 1=&Multiply_Float(A=0.7, B=$Particles.SizeAll)
    Module.Index0Float0Write=true
    Module.Index0Float1Write=true
    Module.Index0Float2Write=true
    Module.Index0Float3Write=true
   )
   &InitialMeshOrientation(Mesh Orientation Mode=::NewEnumerator7, Rotation=<-0.06 0 0>, Use Rotation Vector=true) {module=InitialMeshOrientation001}
  ]
  particleUpdateScript [
   &ParticleState(Kill Particles When Lifetime Has Elapsed=false)
   &ScaleMeshSize(Scale Factor=&Multiply_VectorByFloat(Vector=<1 1 1>, Float=$Particles.SizeAll))
   &CameraOffset()
  ]
  rendererClasses [&NiagaraMeshRendererProperties]
  renderers [
   {rendererClass=&NiagaraMeshRendererProperties, rendererIndex=0}
   {rendererClass=&NiagaraMeshRendererProperties, rendererIndex=1}
   {rendererClass=&NiagaraMeshRendererProperties, rendererIndex=2}
  ]
 }
}
summary {
 systemName=NS_Pickaxe_SheerWill_Idle
 emitters [
  {bEnabled=false, emitterName=LeftSpline, rendererClasses [&NiagaraMeshRendererProperties]}
  {bEnabled=false, emitterName=RightSpline, rendererClasses [&NiagaraMeshRendererProperties]}
  {emitterName=BottomGlow, rendererClasses [&NiagaraMeshRendererProperties]}
  {emitterName=Dragonsparks, rendererClasses [&NiagaraSpriteRendererProperties]}
  {emitterName=ParticleDecal, rendererClasses [&NiagaraSpriteRendererProperties, &NiagaraLightRendererProperties]}
  {bEnabled=false, emitterName=RightSpline001, rendererClasses [&NiagaraMeshRendererProperties]}
  {emitterName=CenterSpline, rendererClasses [&NiagaraMeshRendererProperties]}
  {emitterName=LeftSpline001, rendererClasses [&NiagaraMeshRendererProperties]}
  {bEnabled=false, emitterName=CenterSpline001, rendererClasses [&NiagaraMeshRendererProperties]}
 ]
 userVariables [
  User.bIsFrontEnd {type=NiagaraBool}
  User.bIsFrontEndPreview {type=NiagaraBool}
  User.VariantColorIndex {type=NiagaraFloat, defaultValue=0, description=""}
 ]
}
