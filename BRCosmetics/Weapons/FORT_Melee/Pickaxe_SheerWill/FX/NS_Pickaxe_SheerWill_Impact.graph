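# Niagara impact effect for the Sheer Will pickaxe, layering a center trail mesh, flares with a light and mesh, sparks, fragments and two spike sprite bursts.
# Directional emitters launch along a system-level impact direction, and sizes and speeds scale with a shared SizeAll value.
# User.VariantColorIndex picks among Blot, Citadel and Froth colors and intensities through scratch dynamic inputs.
# Front-end, platform and time-of-day helpers adjust size, camera offset and brightness.
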NiagaraSystem /BRCosmetics/Weapons/FORT_Melee/Pickaxe_SheerWill/FX/NS_Pickaxe_SheerWill_Impact
dependencies {
 usedDataInterfaces [
  &NiagaraDataInterfaceCurve
  &NiagaraDataInterfaceVector2DCurve
  &NiagaraDataInterfaceSpriteRendererInfo
  &NiagaraDataInterfacePlatformSet
 ]
 usedDynamicInputs [
  &SimulationPosition
  /Game/Effects/Niagara/DynamicInputs/FN_DirectionalVectorfromMontageIndex
  &self:ScratchDynamicInput_01
  &self:ScratchDynamicInput_02
  &self:ScratchDynamicInput_03
  &self:ScratchDynamicInput_04
 ]
 usedModules [
  &self:SystemSpawnScript.SystemScriptSource.SystemScriptGraph.NiagaraNodeAssignment_0.SetVariables_#1
  &self:SystemSpawnScript.SystemScriptSource.SystemScriptGraph.NiagaraNodeAssignment_1.SetVariables_#2
  &SystemState
  &self:SystemSpawnScript.SystemScriptSource.SystemScriptGraph.NiagaraNodeAssignment_2.SetVariables_#3
  &self:SystemSpawnScript.SystemScriptSource.SystemScriptGraph.NiagaraNodeAssignment_3.SetVariables_#4
  &self:SystemSpawnScript.SystemScriptSource.SystemScriptGraph.NiagaraNodeAssignment_4.SetVariables_#5
  &self:SystemSpawnScript.SystemScriptSource.SystemScriptGraph.NiagaraNodeAssignment_5.SetVariables_#6
 ]
}
emitters {
 CenterTrail {
  emitterUpdateScript [
   /SetVariables(Emitter.Des:NiagaraFloat=&self:ScratchDynamicInput_08(
     Blot:NiagaraFloat=0.5
     Citadel:NiagaraFloat=0.5
     Froth:NiagaraFloat=1
     VariantColorIndex:NiagaraInt32=&MakeIntFromFloat(Float=$User.VariantColorIndex)
    )) {isSetParameters=true}
   &EmitterState(Life Cycle Mode=::NewEnumerator1, Loop Behavior=::NewEnumerator1, Loop Duration=1, UseLoopDelay=false)
   &SpawnBurst_Instantaneous(Use Spawn Probability=false, UseLoopCountLimit=false)
  ]
  particleSpawnScript [
   &InitializeParticle(
    Lifetime=0.4
    Mesh Scale Mode=::NewEnumerator3
    Mesh Uniform Scale=1.7
    Position Offset=<5 0 0>
    Sprite Size Mode=::NewEnumerator3
    Uniform Sprite Size=50
    UsePositionOffset=true
    Write Material Random=true
   )
   &InitialMeshOrientation(Mesh Orientation Mode=::NewEnumerator7, Rotation=<0 0 0.25>, Use Rotation Vector=true)
   &InitialMeshOrientation(Mesh Orientation Mode=::NewEnumerator2, Rotation=<0 0 -0.25>, Use Rotation Vector=true) {module=InitialMeshOrientation001}
   &AddVelocity(Velocity=$System.Impactdirection)
   &CameraOffset(Camera Offset Amount=30)
  ]
  particleUpdateScript [
   /SetVariables(Particles.AlphaAll:NiagaraFloat=&FloatFromCurve(
     FloatCurve=DataInterface() {propertyValues {Curve {keys [<at 0 1>, {arriveTangent=-2.1124, interpMode=RCIM_Cubic, leaveTangent=-2.1124, tangentMode=RCTM_User, time=0.3473, value=0.8682}, <at 1 0>]}}}
     Scale Curve=&Multiply_Float(
      A=2
      B=&self:ScratchDynamicInput_06(
       Blot:NiagaraFloat=1
       Citadel:NiagaraFloat=1
       Froth:NiagaraFloat=2
       VariantColorIndex:NiagaraInt32=&MakeIntFromFloat(Float=$User.VariantColorIndex)
      )
     )
    )) {isSetParameters=true}
   &FN_Mesh_Attributes(
    Color=&MakeLinearColorFromVectorAndFloat(
     "Vector (RGB)"=&Multiply_Vector(
      A=&MakeVectorFromLinearColorRGB(Color=&self:ScratchDynamicInput_05(
        BlotColor:LinearColor=<rgba 1 0.0146 0 0.4>
        CitadelColor:LinearColor=<rgba 0 0.3333 1 0.4>
        FrothColor:LinearColor=<rgba 0 0 0 1>
        VariantColorIndex:NiagaraInt32=&MakeIntFromFloat(Float=$User.VariantColorIndex)
       ))
      B=<1 1 1>
     )
     "Float (Alpha)"=&Multiply_Float(A=0.7, B=$Particles.AlphaAll)
    )
    Create Color Attribute=true
    Module Output=::NewEnumerator1
   )
   &ParticleState()
   &ScaleMeshSize(Scale Factor=&Multiply_VectorByFloat(
     Vector=<1.4 1 1>
     Float=&FloatFromCurve(
      FloatCurve=DataInterface() {propertyValues {Curve {defaultValue=1, keys [{tangentMode=RCTM_User, time=0, value=0}, {tangentMode=RCTM_User, time=0.3, value=1}]}}}
      Scale Curve=$System.SizeAll
     )
    ))
   &DynamicMaterialParameters(
    Index 0 Param 1=&Multiply_Float(A=1.4, B=$System.SizeAll)
    Index 0 Param 2=0.4
    Index 0 Param 3=&FloatFromCurve(FloatCurve=DataInterface() {propertyValues {Curve {defaultValue=1, keys [{tangentMode=RCTM_User, time=0, value=0.3}, {tangentMode=RCTM_User, time=1, value=0.8}]}}})
    Module.Index0Float0Write=true
    Module.Index0Float1Write=true
    Module.Index0Float2Write=true
    Module.Index0Float3Write=true
   )
   &SolveForcesAndVelocity(Clamp Velocity=false, Limit Acceleration=false)
   &ScaleColor(
    Scale Alpha=$Particles.AlphaAll
    Scale RGB=&Multiply_VectorByFloat(Vector=<1 1 1>, Float=1)
    ScaleA=true
    ScaleRGB=true
   )
  ]
  rendererClasses [&NiagaraMeshRendererProperties]
  renderers [
   {rendererClass=&NiagaraMeshRendererProperties, rendererIndex=0}
   {rendererClass=&NiagaraMeshRendererProperties, rendererIndex=1}
  ]
 }
 DirectionalSparks001 {
  emitterUpdateScript [
   &EmitterState()
   &SpawnBurst_Instantaneous(Spawn Count=20, Use Spawn Probability=true, UseLoopCountLimit=false)
  ]
  particleSpawnScript [
   &InitializeParticle(
    Color=&self:ScratchDynamicInput_05(
     BlotColor:LinearColor=<rgba 5 0.7135 0.7135 1>
     CitadelColor:LinearColor=<rgba 0 1.6667 5 1>
     FrothColor:LinearColor=<rgba 0 0 0 100>
     VariantColorIndex:NiagaraInt32=&MakeIntFromFloat(Float=$User.VariantColorIndex)
    )
    Lifetime Max=0.7
    Lifetime Min=0.3
    Lifetime Mode=::NewEnumerator1
    Sprite Size=&Multiply_Vector2D_ByFloat(
     Float=&Multiply_Float(
      A=&FN_FrontEndFloatModifier(Gameplay=1, Front End=1)
      B=&RandomRangeFloat(Minimum=4.5, Maximum=5)
     )
     Vector2D=<1 1.5>
    )
    Sprite Size Mode=::NewEnumerator4
    UsePositionOffset=false
    Write Material Random=true
   )
   &CameraOffset(Camera Offset Amount=40)
   &AddVelocity(
    Cone Angle=40
    Cone Axis=$System.Impactdirection
    Inner Cone Angle=20
    Module.UseConeFalloff=false
    Module.UseVelDistribution=false
    Override Seed=false
    Velocity Mode=::NewEnumerator2
    Velocity Speed=&Multiply_Float(A=&RandomRangeFloat(Minimum=100, Maximum=300), B=$System.SizeAll)
   )
  ]
  particleUpdateScript [
   &ParticleState()
   &SolveForcesAndVelocity(Clamp Velocity=false, Limit Acceleration=false)
   &ScaleSpriteSizeBySpeed(Max Scale Factor=<0.2 10>, Min Scale Factor=<0.2 0.2>)
   &ScaleSpriteSize(
    Uniform Curve Scale=&Multiply_Float(A=1, B=$System.SizeAll)
    Uniform Curve Sprite Scale=DataInterface() {propertyValues {Curve {keys [{arriveTangent=-6.725, interpMode=RCIM_Cubic, leaveTangent=-6.725, tangentMode=RCTM_User, time=0.65, value=1}, <at 1 0 RCIM_Cubic>]}}}
   )
   &ScaleColor(
    Scale RGB=&VectorFromFloat(Value=&FloatFromCurve(
      FloatCurve=DataInterface() {propertyValues {Curve {keys [
          {interpMode=RCIM_Cubic, tangentMode=RCTM_User, time=0, value=1}
          <at 0.3 0.2>
          {interpMode=RCIM_Cubic, tangentMode=RCTM_User, time=1, value=0}
         ]}}}
      Scale Curve=10
     ))
    ScaleA=true
    ScaleRGB=true
   )
   /SetVariables(
    Particles.GlowColor=&MakeLinearColorFromVectorAndFloat(
     "Vector (RGB)"=&MakeVectorFromLinearColorRGB(Color=&self:ScratchDynamicInput_05(
       BlotColor:LinearColor=<rgba 1 0.05 0.05 1>
       CitadelColor:LinearColor=<rgba 0 0.3333 1 1>
       FrothColor:LinearColor=<rgba 0 0 0 1>
       VariantColorIndex:NiagaraInt32=&MakeIntFromFloat(Float=$User.VariantColorIndex)
      ))
     "Float (Alpha)"=&FloatFromCurve(
      FloatCurve=DataInterface() {propertyValues {Curve {keys [<at 0.5 1>,<at 1 0>]}}}
      Scale Curve=&Multiply_Float(
       A=0.2
       B=&self:ScratchDynamicInput_06(
        Blot:NiagaraFloat=1
        Citadel:NiagaraFloat=1
        Froth:NiagaraFloat=2
        VariantColorIndex:NiagaraInt32=&MakeIntFromFloat(Float=$User.VariantColorIndex)
       )
      )
     )
    )
    Particles.GlowSize=&Multiply_Vector2D_ByFloat(Float=$System.SizeAll, Vector2D=&Multiply_Vector2D(A=$Particles.SpriteSize, B=<8 2>))
   ) {isSetParameters=true}
  ]
  rendererClasses [&NiagaraSpriteRendererProperties]
  renderers [
   {rendererClass=&NiagaraSpriteRendererProperties}
   {rendererClass=&NiagaraSpriteRendererProperties, rendererIndex=1}
  ]
 }
 Flare002 {
  emitterUpdateScript [
   &EmitterState(Life Cycle Mode=::NewEnumerator1, Loop Behavior=::NewEnumerator1, Loop Duration=1, UseLoopDelay=false)
   &SpawnBurst_Instantaneous(Spawn Count=1, Use Spawn Probability=false)
  ]
  particleSpawnScript [
   &InitializeParticle(
    Color Mode=::NewEnumerator0
    Lifetime=0.4
    Sprite Size Mode=::NewEnumerator3
    Uniform Sprite Size=300
    UsePositionOffset=false
    Write Material Random=true
   )
   &DynamicMaterialParameters(
    Index 0 Param 1=&RandomRangeFloat(Minimum=0.4, Maximum=0.6)
    Index 0 Param 2=1
    Index 0 Param 3=1
    Index 0 Param 4=1
    Module.Param0WriteEnabled=true
    Module.Param1WriteEnabled=true
    Module.Param2WriteEnabled=true
    Module.Param3WriteEnabled=true
   )
   &SpriteFacingAndAlignment(
    Alignment Coordinate Space=::Local
    Facing Coordinate Space=::Local
    Sprite Alignment=<0 0 1>
    Sprite Facing=<0 0 1>
    WriteAlignment=true
    WriteFacing=true
   )
   &AddVelocity(Velocity=$System.Impactdirection)
  ]
  particleUpdateScript [
   &ParticleState()
   &SolveForcesAndVelocity(Clamp Velocity=false, Limit Acceleration=false)
   &CameraOffset(Camera Offset Amount=40)
   &FN_Sprite_Attributes(
    Color=&MakeLinearColorFromVectorAndFloat(
     "Vector (RGB)"=&Multiply_VectorByFloat(
      Vector=&MakeVectorFromLinearColorRGB(Color=&self:ScratchDynamicInput_05(
        BlotColor:LinearColor=<rgba 20 0 0.3333 1>
        CitadelColor:LinearColor=<rgba 0 0.8333 5 1>
        FrothColor:LinearColor=<rgba 0.2 0.2 0.2 1>
        VariantColorIndex:NiagaraInt32=&MakeIntFromFloat(Float=$User.VariantColorIndex)
       ))
      Float=4
     )
     "Float (Alpha)"=&Multiply_Float(
      A=1
      B=&FloatFromCurve(
       FloatCurve=DataInterface() {propertyValues {Curve {keys [{arriveTangent=-1.3786, interpMode=RCIM_Cubic, leaveTangent=-1.3786, tangentMode=RCTM_User, time=0, value=1}, <at 0.42 0>]}}}
       Scale Curve=10
      )
     )
    )
    Create Color Attribute=true
    Create SpriteSize Attribute=true
    SpriteSize=&Multiply_Vector2D_ByFloat(
     Float=&FloatFromCurve(
      FloatCurve=DataInterface() {propertyValues {Curve {keys [
          {arriveTangent=4.1785, interpMode=RCIM_Cubic, leaveTangent=4.1785, tangentMode=RCTM_User, time=0, value=0}
          {arriveTangent=1.1179, interpMode=RCIM_Cubic, leaveTangent=1.1179, tangentMode=RCTM_User, time=0.0858, value=0.6228}
          {arriveTangent=0.0224, interpMode=RCIM_Cubic, leaveTangent=0.0224, tangentMode=RCTM_User, time=0.42, value=1}
         ]}}}
      Scale Curve=&Multiply_Float(A=&FN_FrontEndFloatModifier(Gameplay=0.8, Front End=0.7), B=$System.SizeAll)
     )
     Vector2D=<60 150>
    )
    Write to Module Namespace=true
   ) {module=FN_Sprite_Attributes001}
  ]
  rendererClasses [&NiagaraSpriteRendererProperties]
  renderers [{rendererClass=&NiagaraSpriteRendererProperties}]
 }
 Fragment {
  emitterUpdateScript [
   &EmitterState()
   &SpawnBurst_Instantaneous(Spawn Count=6, Use Spawn Probability=true, UseLoopCountLimit=false)
  ]
  particleSpawnScript [
   &InitializeParticle(
    Color=&self:ScratchDynamicInput_05(
     BlotColor:LinearColor=<rgba 1 0.291 0.2049 1>
     CitadelColor:LinearColor=<rgba 0 1 3 1>
     FrothColor:LinearColor=<rgba 0 0 0 2>
     VariantColorIndex:NiagaraInt32=&MakeIntFromFloat(Float=$User.VariantColorIndex)
    )
    Lifetime Max=0.6
    Lifetime Min=0.4
    Lifetime Mode=::NewEnumerator1
    Mesh Scale Mode=::NewEnumerator1
    Mesh Uniform Scale Max=1.8
    UsePositionOffset=false
    Write Material Random=true
   )
   &CameraOffset(Camera Offset Amount=40)
   &AddVelocity(
    Cone Angle=50
    Cone Axis=$System.Impactdirection
    Inner Cone Angle=30
    Module.UseConeFalloff=false
    Module.UseVelDistribution=false
    Override Seed=false
    Velocity Mode=::NewEnumerator2
    Velocity Speed=&Multiply_Float(A=&RandomRangeFloat(Minimum=100, Maximum=400), B=$System.SizeAll)
   )
   &InitialMeshOrientation(Use Rotation Vector=false)
  ]
  particleUpdateScript [
   &ParticleState()
   &ScaleMeshSize(Scale Factor=&Multiply_VectorByFloat(
     Vector=&Multiply_VectorByFloat(Vector=<1 1 1>, Float=$System.SizeAll)
     Float=&FloatFromCurve(FloatCurve=DataInterface() {propertyValues {Curve {keys [<at 0 0>,<at 0.1 1.0029>,<at 0.7 1>,<at 1 0>]}}}, Scale Curve=1.2)
    ))
   &UpdateMeshOrientation(
    Rotation Rate=&RandomRangeFloat(Minimum=2, Maximum=2)
    Rotation Vector=&RandomRangeVector(Minimum=<-1 -1 -1>, Maximum=<1 1 1>)
   )
   &Drag(Drag=0.5, Use Linear Drag=true, Use Rotational Drag=true)
   &SolveForcesAndVelocity(Clamp Velocity=false, Limit Acceleration=false)
   &ScaleColor(
    Scale Alpha=&FloatFromCurve(FloatCurve=DataInterface() {propertyValues {Curve {keys [<at 0 1>,<at 0.8 1>,<at 1 0>]}}})
    Scale RGB=&VectorFromFloat(Value=&FloatFromCurve(
      FloatCurve=DataInterface() {propertyValues {Curve {keys [
          {interpMode=RCIM_Cubic, tangentMode=RCTM_User, time=0, value=0.2}
          <at 0.5012 0.2235>
          {interpMode=RCIM_Cubic, tangentMode=RCTM_User, time=1, value=0.2}
         ]}}}
      Scale Curve=&RandomRangeFloat(Minimum=3, Maximum=10)
     ))
    ScaleA=true
    ScaleRGB=true
   )
  ]
  rendererClasses [&NiagaraMeshRendererProperties]
  renderers [{rendererClass=&NiagaraMeshRendererProperties, rendererIndex=0}]
 }
 Impact_Flare001 {
  emitterUpdateScript [
   &EmitterState()
   &SpawnBurst_Instantaneous(Spawn Count=1, Use Spawn Probability=false)
  ]
  particleSpawnScript [
   &InitializeParticle(
    Color=&self:ScratchDynamicInput_05(
     BlotColor:LinearColor=<rgba 5 0 0 2>
     CitadelColor:LinearColor=<rgba 0 1.6667 5 2>
     FrothColor:LinearColor=<rgba 5 4.7227 4 1>
     VariantColorIndex:NiagaraInt32=&MakeIntFromFloat(Float=$User.VariantColorIndex)
    )
    Lifetime=0.25
    Mass Mode=::NewEnumerator1
    Mesh Scale=&UniformRangedVector(Minimum=<1.25 1.25 1.25>, Maximum=<1.4 1.4 1.4>)
    Mesh Scale Mode=::NewEnumerator4
    Position Offset=<0 0 2.5>
    Sprite Rotation Mode=::NewEnumerator2
    Sprite Size Mode=::NewEnumerator3
    Uniform Sprite Size=&FN_FrontEndFloatModifier(Gameplay=180, Front End=70)
    UsePositionOffset=true
    Write Material Random=true
   ) {module=InitializeParticle001}
   &CameraOffset(Camera Offset Amount=&FN_PickFloatDependOnPlatform(TargetPlatformActive=10, TargetPlatformNoActive=30))
  ]
  particleUpdateScript [
   &ParticleState()
   &ScaleSpriteSize(
    Scale Sprite Size Mode=::NewEnumerator0
    Uniform Scale Factor=&Multiply_Float(
     A=&Multiply_Float(A=0.8, B=$System.SizeAll)
     B=&FN_PickFloatDependOnPlatform(TargetPlatformActive=1, TargetPlatformNoActive=0.7)
    )
   )
   &DynamicMaterialParameters(
    Index 0 Param 1=&FloatFromCurve(FloatCurve=DataInterface() {propertyValues {Curve {keys [<at 0 1 RCIM_Cubic>,<at 0.8 0 RCIM_Cubic>]}}})
    Index 0 Param 2=2
    Index 0 Param 3=1
    Index 0 Param 4=1
    Index 1 Param 1=&FloatFromCurve(FloatCurve=DataInterface() {propertyValues {Curve {keys [
         {arriveTangent=-2.3163, interpMode=RCIM_Cubic, time=0, value=2}
         {arriveTangent=-1.9019, interpMode=RCIM_Cubic, leaveTangent=-1.9019, tangentMode=RCTM_User, time=1, value=0.4}
        ]}}})
    Index 1 Param 2=0.4
    Index 1 Param 3=0.6
    Index 1 Param 4=1
    Index 2 Param 1=&FloatFromCurve(FloatCurve=DataInterface() {propertyValues {Curve {keys [<at 0.6 0>,<at 1 1>]}}})
    Index 2 Param 2=45
    Module.Param0WriteEnabled=true
    Module.Param0WriteEnabled1=false
    Module.Param0WriteEnabled2=true
    Module.Param1WriteEnabled=true
    Module.Param1WriteEnabled1=true
    Module.Param1WriteEnabled2=true
    Module.Param2WriteEnabled=true
    Module.Param2WriteEnabled1=true
    Module.Param2WriteEnabled2=true
    Module.Param3WriteEnabled=true
    Module.Param3WriteEnabled1=true
    Module.Param3WriteEnabled2=true
    Write Parameter Index 1=true
    Write Parameter Index 2=true
   )
   &ScaleColor(
    Scale Alpha=&FloatFromCurve(
     FloatCurve=DataInterface() {propertyValues {Curve {keys [<at 0 1>,<at 0.9 0>]}}}
     Scale Curve=&FNTimeOfDayScalar(Night Scale=0.4, Morning Scale=0.7, Day Scale=0.7, Evening Scale=0.7)
    )
    Scale RGB=&MakeVectorFromLinearColorRGB(Color=<rgba 1 1 1 1>)
    ScaleA=true
    ScaleRGB=true
   )
   /SetVariables(
    Particles.ShadowCamOffset=10
    Particles.ShadowColor:LinearColor=&MakeLinearColorFromVectorAndFloat(
     "Vector (RGB)"=&MakeVectorFromLinearColorRGB(Color=&self:ScratchDynamicInput_05(
       BlotColor:LinearColor=<rgba 0.1562 0.0052 0 1.4>
       CitadelColor:LinearColor=<rgba 0 0.0521 0.1562 1.4>
       FrothColor:LinearColor=<rgba 0 0 0 1>
       VariantColorIndex:NiagaraInt32=&MakeIntFromFloat(Float=$User.VariantColorIndex)
      ))
     "Float (Alpha)"=&FloatFromCurve(
      FloatCurve=DataInterface() {propertyValues {Curve {
         defaultValue=1
         keys [{interpMode=RCIM_Cubic, tangentMode=RCTM_User, time=0, value=1}, {interpMode=RCIM_Cubic, tangentMode=RCTM_User, time=1, value=0}]
        }}}
      Scale Curve=&Multiply_Float(
       A=1
       B=&self:ScratchDynamicInput_06(
        Blot:NiagaraFloat=1
        Citadel:NiagaraFloat=1
        Froth:NiagaraFloat=5
        VariantColorIndex:NiagaraInt32=&MakeIntFromFloat(Float=$User.VariantColorIndex)
       )
      )
     )
    )
    Particles.ShadowSize:Vector2f=&Multiply_Vector2D_ByFloat(
     Float=&Multiply_Float(
      A=&FN_PickFloatDependOnPlatform(TargetPlatformActive=0.7, TargetPlatformNoActive=0.7)
      B=$System.SizeAll
     )
     Vector2D=&Vector2DFromFloat(Value=&FN_FrontEndFloatModifier(Gameplay=160, Front End=50))
    )
   ) {isSetParameters=true}
   &FN_Light_Attributes(
    Color=&self:ScratchDynamicInput_05(
     BlotColor:LinearColor=<rgba 5 0 0 2>
     CitadelColor:LinearColor=<rgba 0 1.6667 5 2>
     FrothColor:LinearColor=<rgba 2 1.8891 1.6 1>
     VariantColorIndex:NiagaraInt32=&MakeIntFromFloat(Float=$User.VariantColorIndex)
    )
    Create Color Attribute=true
    Create Exponent Attribute=true
    Create Radius Attribute=true
    Exponent=3
    Radius=110
    Scale Color=&FloatFromCurve(
     FloatCurve=DataInterface() {propertyValues {Curve {keys [
         {interpMode=RCIM_Cubic, leaveTangent=22.4876, tangentMode=RCTM_Break, time=0, value=0.4}
         {interpMode=RCIM_Cubic, tangentMode=RCTM_Break, time=0.1, value=1.4}
         {arriveTangent=-5.6532, interpMode=RCIM_Cubic, leaveTangent=-2.1385, tangentMode=RCTM_Break, time=0.3, value=1}
         {arriveTangent=-0.6872, tangentMode=RCTM_User, time=1, value=0}
        ]}}}
     Scale Curve=8
    )
   )
   &FN_Mesh_Attributes(
    Color=&MakeLinearColorFromVectorAndFloat(
     "Vector (RGB)"=&MakeVectorFromLinearColorRGB(Color=&self:ScratchDynamicInput_05(
       BlotColor:LinearColor=<rgba 1 0.0427 0.0427 1>
       CitadelColor:LinearColor=<rgba 0.0427 0.3618 1 1>
       FrothColor:LinearColor=<rgba 0 0 0 1>
       VariantColorIndex:NiagaraInt32=&MakeIntFromFloat(Float=$User.VariantColorIndex)
      ))
     "Float (Alpha)"=&Multiply_Float(
      A=&FN_PickFloatDependOnPlatform(TargetPlatformActive=1, TargetPlatformNoActive=0.5)
      B=&self:ScratchDynamicInput_06(
       Blot:NiagaraFloat=1
       Citadel:NiagaraFloat=1
       Froth:NiagaraFloat=1
       VariantColorIndex:NiagaraInt32=&MakeIntFromFloat(Float=$User.VariantColorIndex)
      )
     )
    )
    Create Color Attribute=true
    Create Scale Attribute=true
    Module Output=::NewEnumerator1
    Scale=&Multiply_VectorByFloat(
     Vector=<1 1 1>
     Float=&FloatFromCurve(
      FloatCurve=DataInterface() {propertyValues {Curve {
         defaultValue=1
         keys [
          {arriveTangent=1.3753, interpMode=RCIM_Cubic, leaveTangent=1.3753, tangentMode=RCTM_User, time=0, value=0}
          {tangentMode=RCTM_User, time=1, value=1}
         ]
        }}}
      Scale Curve=&Multiply_Float(A=1.4, B=$System.SizeAll)
     )
    )
   )
  ]
  rendererClasses [
   &NiagaraSpriteRendererProperties
   &NiagaraLightRendererProperties
   &NiagaraMeshRendererProperties
  ]
  renderers [
   {rendererClass=&NiagaraSpriteRendererProperties}
   {rendererClass=&NiagaraSpriteRendererProperties, rendererIndex=1}
   {rendererClass=&NiagaraLightRendererProperties, rendererIndex=2}
   {rendererClass=&NiagaraMeshRendererProperties, rendererIndex=3}
  ]
 }
 spikes {
  emitterUpdateScript [
   &EmitterState(Life Cycle Mode=::NewEnumerator1, Loop Behavior=::NewEnumerator1, Loop Delay=0.03, Loop Duration=0.6, UseLoopDelay=false)
   &SpawnBurst_Instantaneous(Spawn Count=3, Use Spawn Probability=false, UseLoopCountLimit=false)
  ]
  particleSpawnScript [
   &InitializeParticle(
    Color=&self:ScratchDynamicInput_05(
     BlotColor:LinearColor=<rgba 4 0.0681 0.0681 0.5>
     CitadelColor:LinearColor=<rgba 0.0681 1.3787 4 0.5>
     FrothColor:LinearColor=<rgba 0 0 0 1>
     VariantColorIndex:NiagaraInt32=&MakeIntFromFloat(Float=$User.VariantColorIndex)
    )
    Lifetime Max=0.6
    Lifetime Min=0.4
    Lifetime Mode=::NewEnumerator1
    Sprite Size Max=<70 100>
    Sprite Size Min=<50 100>
    Sprite Size Mode=::NewEnumerator2
    UsePositionOffset=false
    Write Material Random=true
   )
   &TorusLocation(Align Torus With Axis=false, Handle Radius=0, Large Radius=12, MaskSpawnGroup=false, Module.Override Randomness=false)
   &SubUVAnimation(End Frame=0, SubUV Animation Mode:ENiagara_SubUVLookupMode=::NewEnumerator2)
   &AddVelocity(
    Cone Angle=30
    Cone Axis=$System.Impactdirection
    Inner Cone Angle=30
    Module.UseConeFalloff=false
    Module.UseVelDistribution=false
    Override Seed=false
    Velocity Mode=::NewEnumerator2
    Velocity Speed=&Multiply_Float(A=&RandomRangeFloat(Minimum=200, Maximum=300), B=$System.SizeAll)
   )
   /SetVariables(Particles.SizeAll:NiagaraFloat=&Multiply_Float(A=0.35, B=$System.SizeAll)) {isSetParameters=true, module=SetVariables_#7}
   &CameraOffset(Camera Offset Amount=30)
  ]
  particleUpdateScript [
   &DynamicMaterialParameters(
    Index 0 Param 1=3
    Index 0 Param 2=10
    Index 0 Param 4=&FloatFromCurve(FloatCurve=DataInterface() {propertyValues {Curve {keys [<at 0.4 0>,<at 1 1>]}}})
    Module.Index0Float0Write=true
    Module.Index0Float1Write=true
    Module.Index0Float2Write=true
    Module.Index0Float3Write=true
   )
   &ParticleState()
   &ScaleSpriteSize(
    Uniform Curve Scale=$Particles.SizeAll
    Uniform Curve Sprite Scale=DataInterface() {propertyValues {Curve {keys [{tangentMode=RCTM_User, time=0, value=1}, {tangentMode=RCTM_User, time=1, value=1.5}]}}}
   ) {module=ScaleSpriteSize001}
   &Drag(
    Drag=&FloatFromCurve(FloatCurve=DataInterface() {propertyValues {Curve {keys [<at 0 6>,<at 0.2 8>]}}}, CurveIndex=$Emitter.Age, Scale Curve=0.5)
    Use Linear Drag=true
    Use Rotational Drag=false
   )
   &SolveForcesAndVelocity(Clamp Velocity=false, Limit Acceleration=false)
   &ScaleSpriteSize(
    Non-Uniform Curve Scale=&Multiply_Vector2D_ByFloat(Float=&FN_FrontEndFloatModifier(Gameplay=0.6, Front End=0.5), Vector2D=<2.5 3>)
    Non-Uniform Curve Sprite Scale=DataInterface() {propertyValues {
      XCurve {keys [{tangentMode=RCTM_User, time=0, value=0.7}, {tangentMode=RCTM_User, time=1, value=1}]}
      YCurve {keys [{tangentMode=RCTM_User, time=0, value=1}, {tangentMode=RCTM_User, time=1, value=0.6}]}
     }}
    Scale Sprite Size Mode=::NewEnumerator3
   )
   &ScaleColor(
    Scale Alpha=&FloatFromCurve(
     FloatCurve=DataInterface() {propertyValues {Curve {keys [<at 0 0>, <at 0.1996 1>, {arriveTangent=-1.2493, interpMode=RCIM_Cubic, leaveTangent=-1.2493, time=0.5272, value=0.6235}, <at 1 0>]}}}
     Scale Curve=5
    )
    Scale RGB=&Multiply_VectorByFloat(Vector=<1 1 1>, Float=0.01)
    ScaleA=true
    ScaleRGB=true
   )
   &SubUVAnimation(
    "Sprite Renderer "=DataInterface() {propertyValues {SpriteRenderer=&self:spikes_1.NiagaraSpriteRendererProperties_0}}
    UseEndFrame=false
    UseStartFrame=false
   ) {module=SubUVAnimation001}
   &FN_Mesh_Attributes(
    Color=&MakeLinearColorFromVectorAndFloat(
     "Vector (RGB)"=&MakeVectorFromLinearColorRGB(Color=&self:ScratchDynamicInput_05(
       BlotColor:LinearColor=<rgba 0.4 0 0 1>
       CitadelColor:LinearColor=<rgba 0 0.2333 0.7 1>
       FrothColor:LinearColor=<rgba 0.5 0.5 0.5 1>
       VariantColorIndex:NiagaraInt32=&MakeIntFromFloat(Float=$User.VariantColorIndex)
      ))
     "Float (Alpha)"=&Multiply_Float(
      A=2
      B=&FloatFromCurve(FloatCurve=DataInterface() {propertyValues {Curve {keys [<at 0 0>,<at 0.1996 1>,<at 0.534 0.6824>,<at 1 0>]}}})
     )
    )
    Create Color Attribute=true
    Write to Module Namespace=true
   )
   /SetVariables() {isSetParameters=true, module=SetVariables_#8}
   &FN_Sprite_Attributes(
    Color=&MakeLinearColorFromVectorAndFloat(
     "Vector (RGB)"=&MakeVectorFromLinearColorRGB(Color=&self:ScratchDynamicInput_05(
       BlotColor:LinearColor=<rgba 0.5 0 0.0109 1>
       CitadelColor:LinearColor=<rgba 0 0.2667 0.8 1>
       FrothColor:LinearColor=<rgba 0 0 0 1>
       VariantColorIndex:NiagaraInt32=&MakeIntFromFloat(Float=$User.VariantColorIndex)
      ))
     "Float (Alpha)"=&FloatFromCurve(
      FloatCurve=DataInterface() {propertyValues {Curve {keys [<at 0 0>,<at 0.1007 0.9437>,<at 0.7189 0.9362>,<at 0.9 0>]}}}
      Scale Curve=&Multiply_Float(
       A=0.5
       B=&self:ScratchDynamicInput_06(Blot:NiagaraFloat=1, Citadel:NiagaraFloat=1, Froth:NiagaraFloat=2, VariantColorIndex:NiagaraInt32=0)
      )
     )
    )
    Create Color Attribute=true
    Create SpriteSize Attribute=true
    SpriteSize=&Multiply_Vector2D_ByFloat(Float=$Particles.SizeAll, Vector2D=<120 220>)
    Write to Module Namespace=true
   )
  ]
  rendererClasses [&NiagaraSpriteRendererProperties]
  renderers [
   {rendererClass=&NiagaraSpriteRendererProperties}
   {rendererClass=&NiagaraSpriteRendererProperties, rendererIndex=1}
   {rendererClass=&NiagaraSpriteRendererProperties, rendererIndex=2}
  ]
 }
 spikesFromCenter {
  emitterUpdateScript [
   &EmitterState(Life Cycle Mode=::NewEnumerator1, Loop Behavior=::NewEnumerator1, Loop Delay=0.03, Loop Duration=0.6, UseLoopDelay=false)
   &SpawnBurst_Instantaneous(Spawn Count=6, Use Spawn Probability=false, UseLoopCountLimit=false)
  ]
  particleSpawnScript [
   &InitializeParticle(
    Color=&self:ScratchDynamicInput_05(
     BlotColor:LinearColor=<rgba 2 0.034 0.1979 0.5>
     CitadelColor:LinearColor=<rgba 0.034 0.6894 2 0.5>
     FrothColor:LinearColor=<rgba 0.0052 0.0052 0.0052 1>
     VariantColorIndex:NiagaraInt32=&MakeIntFromFloat(Float=$User.VariantColorIndex)
    )
    Lifetime Max=0.3
    Lifetime Min=0.2
    Lifetime Mode=::NewEnumerator1
    Sprite Size Max=<70 100>
    Sprite Size Min=<50 100>
    Sprite Size Mode=::NewEnumerator2
    UsePositionOffset=false
    Write Material Random=true
   )
   &SubUVAnimation(End Frame=0, SubUV Animation Mode:ENiagara_SubUVLookupMode=::NewEnumerator2)
   &AddVelocity(
    Module.Use Falloff Distance=false
    Override Seed=false
    Velocity Mode=::NewEnumerator1
    Velocity Speed=&Multiply_Float(A=&RandomRangeFloat(Minimum=200, Maximum=200), B=$System.SizeAll)
   )
   /SetVariables(Particles.SizeAll:NiagaraFloat=&Multiply_Float(A=0.25, B=$System.SizeAll)) {isSetParameters=true, module=SetVariables_#7}
   &CameraOffset(Camera Offset Amount=30)
  ]
  particleUpdateScript [
   &DynamicMaterialParameters(
    Index 0 Param 1=3
    Index 0 Param 2=10
    Index 0 Param 4=&FloatFromCurve(FloatCurve=DataInterface() {propertyValues {Curve {keys [<at 0.4 0>,<at 1 1>]}}})
    Module.Index0Float0Write=true
    Module.Index0Float1Write=true
    Module.Index0Float2Write=true
    Module.Index0Float3Write=true
   )
   &ParticleState()
   &ScaleSpriteSize(
    Uniform Curve Scale=$Particles.SizeAll
    Uniform Curve Sprite Scale=DataInterface() {propertyValues {Curve {keys [{tangentMode=RCTM_User, time=0, value=1}, {tangentMode=RCTM_User, time=1, value=1.5}]}}}
   ) {module=ScaleSpriteSize001}
   &Drag(Use Linear Drag=true, Use Rotational Drag=false)
   &SolveForcesAndVelocity(Clamp Velocity=false, Limit Acceleration=false)
   &ScaleSpriteSize(
    Non-Uniform Curve Scale=&Multiply_Vector2D_ByFloat(Float=&FN_FrontEndFloatModifier(Gameplay=0.6, Front End=0.5), Vector2D=<2.5 3>)
    Non-Uniform Curve Sprite Scale=DataInterface() {propertyValues {
      XCurve {keys [{tangentMode=RCTM_User, time=0, value=0.7}, {tangentMode=RCTM_User, time=1, value=1}]}
      YCurve {keys [{tangentMode=RCTM_User, time=0, value=1}, {tangentMode=RCTM_User, time=1, value=0.6}]}
     }}
    Scale Sprite Size Mode=::NewEnumerator3
   )
   &ScaleColor(
    Scale Alpha=&FloatFromCurve(
     FloatCurve=DataInterface() {propertyValues {Curve {keys [<at 0 0>, <at 0.1996 1>, {arriveTangent=-1.2493, interpMode=RCIM_Cubic, leaveTangent=-1.2493, time=0.5272, value=0.6235}, <at 1 0>]}}}
     Scale Curve=5
    )
    Scale RGB=&Multiply_VectorByFloat(Vector=<1 1 1>, Float=0.05)
    ScaleA=true
    ScaleRGB=true
   )
   &SubUVAnimation(
    "Sprite Renderer "=DataInterface() {propertyValues {SpriteRenderer=&self:spikesFromCenter_0.NiagaraSpriteRendererProperties_0}}
    UseEndFrame=false
    UseStartFrame=false
   ) {module=SubUVAnimation001}
   /SetVariables() {isSetParameters=true, module=SetVariables_#8}
   &FN_Sprite_Attributes(
    Color=&MakeLinearColorFromVectorAndFloat(
     "Vector (RGB)"=&MakeVectorFromLinearColorRGB(Color=&self:ScratchDynamicInput_05(
       BlotColor:LinearColor=<rgba 0.5 0 0.0109 1>
       CitadelColor:LinearColor=<rgba 0 0.2333 0.7 1>
       FrothColor:LinearColor=<rgba 0 0 0 1>
       VariantColorIndex:NiagaraInt32=&MakeIntFromFloat(Float=$User.VariantColorIndex)
      ))
     "Float (Alpha)"=&FloatFromCurve(
      FloatCurve=DataInterface() {propertyValues {Curve {keys [<at 0 0>,<at 0.1007 0.9437>,<at 0.7189 0.9362>,<at 0.9 0>]}}}
      Scale Curve=&Multiply_Float(
       A=0.5
       B=&self:ScratchDynamicInput_06(Blot:NiagaraFloat=1, Citadel:NiagaraFloat=1, Froth:NiagaraFloat=2, VariantColorIndex:NiagaraInt32=0)
      )
     )
    )
    Create Color Attribute=true
    Create SpriteSize Attribute=true
    SpriteSize=&Multiply_Vector2D_ByFloat(Float=$Particles.SizeAll, Vector2D=<120 220>)
    Write to Module Namespace=true
   )
  ]
  rendererClasses [&NiagaraSpriteRendererProperties]
  renderers [
   {rendererClass=&NiagaraSpriteRendererProperties}
   {rendererClass=&NiagaraSpriteRendererProperties, rendererIndex=1}
  ]
 }
}
summary {
 systemName=NS_Pickaxe_SheerWill_Impact
 emitters [
  {emitterName=CenterTrail, rendererClasses [&NiagaraMeshRendererProperties]}
  {emitterName=Flare002, rendererClasses [&NiagaraSpriteRendererProperties]}
  {emitterName=spikes, rendererClasses [&NiagaraSpriteRendererProperties]}
  {
   emitterName=Impact_Flare001
   rendererClasses [
    &NiagaraSpriteRendererProperties
    &NiagaraLightRendererProperties
    &NiagaraMeshRendererProperties
   ]
  }
  {emitterName=DirectionalSparks001, rendererClasses [&NiagaraSpriteRendererProperties]}
  {emitterName=spikesFromCenter, rendererClasses [&NiagaraSpriteRendererProperties]}
  {emitterName=Fragment, rendererClasses [&NiagaraMeshRendererProperties]}
 ]
 userVariables [
  User.bIsFrontEnd {type=NiagaraBool}
  User.bIsFrontEndPreview {type=NiagaraBool}
  User.bisInVehicle {type=NiagaraBool}
  User.FrontEndScale {type=NiagaraFloat, defaultValue=0.5, description=""}
  User.SwingSectionIndex {type=NiagaraInt32}
  User.VariantColorIndex {type=NiagaraFloat, defaultValue=0, description=""}
 ]
}
