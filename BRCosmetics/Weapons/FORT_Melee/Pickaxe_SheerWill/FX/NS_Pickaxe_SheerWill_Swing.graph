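# Niagara swing effect for the Sheer Will pickaxe, drawing glowing weapon trails with lightning and sparks.
# TrailMain and TrailHighLight are anim-trail ribbons, while Lightning (SubUV) and Dragonsparks (curl noise) spawn sprites from sockets via a system Socket Reader.
# User.VariantColorIndex drives scratch dynamic inputs choosing Blot, Citadel or Froth colors and intensities.
# Dragonsparks glow alpha and TrailMain alpha scale pin that index to 0, ignoring the variant.
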NiagaraSystem /BRCosmetics/Weapons/FORT_Melee/Pickaxe_SheerWill/FX/NS_Pickaxe_SheerWill_Swing
dependencies {
 usedDataInterfaces [
  &NiagaraDataInterfaceCurve
  &NiagaraDataInterfacePlatformSet
  &NiagaraDataInterfaceSpriteRendererInfo
  &NiagaraDataInterfaceSocketReader
 ]
 usedDynamicInputs [
  &self:ScratchDynamicInput_01
  &self:ScratchDynamicInput_02
  &self:ScratchDynamicInput_04
 ]
 usedModules [
  &self:SystemSpawnScript.SystemScriptSource.SystemScriptGraph.NiagaraNodeAssignment_0.SetVariables_#1
  &SystemState
  &self:SystemSpawnScript.SystemScriptSource.SystemScriptGraph.NiagaraNodeAssignment_1.SetVariables_#2
  &self:SystemSpawnScript.SystemScriptSource.SystemScriptGraph.NiagaraNodeAssignment_2.SetVariables_#3
  &self:SystemSpawnScript.SystemScriptSource.SystemScriptGraph.NiagaraNodeAssignment_3.SetVariables_#4
  &self:SystemSpawnScript.SystemScriptSource.SystemScriptGraph.NiagaraNodeAssignment_4.SetVariables_#5
 ]
}
emitters {
 Dragonsparks {
  emitterUpdateScript [
   &EmitterState(Life Cycle Mode=::NewEnumerator1, Loop Behavior=::NewEnumerator1, Loop Duration=0.2, UseLoopDelay=false)
   &SpawnRate(SpawnRate=30, Use Spawn Probability=false)
  ]
  particleSpawnScript [
   &InitializeParticle(
    Color=&self:ScratchDynamicInput_06(
     BlotColor:LinearColor=<rgba 50 0 4.1667 1>
     CitadelColor:LinearColor=<rgba 0 25 50 1>
     FrothColor:LinearColor=<rgba 0 0 0 100>
     VariantColorIndex:NiagaraInt32=&MakeIntFromFloat(Float=$User.VariantColorIndex)
    )
    Lifetime Max=0.7
    Lifetime Min=0.5
    Lifetime Mode=::NewEnumerator1
    Ribbon Width=40
    Ribbon Width Mode=::NewEnumerator1
    Sprite Size Max=<2 7>
    Sprite Size Min=<2 5>
    Sprite Size Mode=::NewEnumerator2
    UsePositionOffset=false
    Write Material Random=true
   )
   &SocketLocation(SamplePosOffset=false, Socket Index=1, Socket Reader=$System.Socket Reader, Socket Sampling Mode=::NewEnumerator2, VelocitySpeedLimit=false)
   /SetVariables(Particles.SizeAll=1.2) {isSetParameters=true}
   &AddVelocity() {module=AddVelocity001}
   &InheritVelocity(Inherited Velocity Amount Scale=<0.1 0.1 0.1>, Speed Limit=true, UseSpeedThreshold=false)
  ]
  particleUpdateScript [
   &ParticleState()
   &ScaleSpriteSize(Scale Factor=&Vector2DFromFloat(Value=&FloatFromCurve(
      FloatCurve=DataInterface() {propertyValues {Curve {keys [
          <at 0 0 RCIM_Cubic>
          {interpMode=RCIM_Cubic, leaveTangent=-3.0296, tangentMode=RCTM_Break, time=0.2, value=1.2}
          {arriveTangent=-1.8107, interpMode=RCIM_Cubic, leaveTangent=-1.8107, tangentMode=RCTM_User, time=1, value=0.2}
         ]}}}
      Scale Curve=&Multiply_Float(A=0.4, B=$Particles.SizeAll)
     ))) {module=ScaleSpriteSize001}
   /SetVariables(Particles.Glow03_Color=&MakeLinearColorFromVectorAndFloat(
     "Vector (RGB)"=&MakeVectorFromLinearColorRGB(Color=&MultiplyLinearColor(
       Linear Color=&self:ScratchDynamicInput_07(
        BlotColor:LinearColor=<rgba 1 0.0112 0.0112 1>
        CitadelColor:LinearColor=<rgba 0.0112 0.3408 1 1>
        FrothColor:LinearColor=<rgba 0 0 0 1>
        VariantColorIndex:NiagaraInt32=&MakeIntFromFloat(Float=$User.VariantColorIndex)
       )
       Scale Factor=&MakeVector4_FromFloat(Float=1)
      ))
     "Float (Alpha)"=&Multiply_Float(
      A=0.1
      B=&self:ScratchDynamicInput_09(Blot:NiagaraFloat=1, Citadel:NiagaraFloat=1, Froth:NiagaraFloat=2, VariantColorIndex:NiagaraInt32=0)
     )
    )) {isSetParameters=true}
   /SetVariables(Particles.Glow03_Size=&Multiply_Vector2D_ByFloat(
     Float=&FloatFromCurve(
      FloatCurve=DataInterface() {propertyValues {Curve {keys [
          <at 0 0 RCIM_Cubic>
          <at 0.2 1.3 RCIM_Cubic>
          {arriveTangent=-0.1423, interpMode=RCIM_Cubic, leaveTangent=-0.1423, tangentMode=RCTM_User, time=1, value=0}
         ]}}}
      Scale Curve=&Multiply_Float(A=5.5, B=$Particles.SizeAll)
     )
     Vector2D=<1 2>
    )) {isSetParameters=true}
   &DynamicMaterialParameters(
    Index 0 Param 1=&MakeCustomFloatFromBool(Bool=&RandomBool(), True Float=0, False Float=0.75)
    Module.Index0Float0Write=true
    Module.Index0Float1Write=true
    Module.Index0Float2Write=true
    Module.Index0Float3Write=true
   )
   &CurlNoiseForce(Bias Noise=false, Noise Frequency=100, Noise Strength=200, Pan Noise=false)
   &Drag(Use Linear Drag=true, Use Rotational Drag=true)
   &SolveForcesAndVelocity(Clamp Velocity=false, Limit Acceleration=false)
  ]
  rendererClasses [&NiagaraSpriteRendererProperties]
  renderers [
   {rendererClass=&NiagaraSpriteRendererProperties}
   {rendererClass=&NiagaraSpriteRendererProperties, rendererIndex=1}
  ]
 }
 Lightning {
  emitterUpdateScript [
   /SetVariables(Emitter.LightningOutline:LinearColor=&self:ScratchDynamicInput_03(
     BlotColor:LinearColor=<rgba 0.3 0 0.01 1>
     CitadelColor:LinearColor=<rgba 0 0.6667 2 1>
     FrothColor:LinearColor=<rgba 0.2 0.2 0.2 1>
     VariantColorIndex:NiagaraInt32=&MakeIntFromFloat(Float=$User.VariantColorIndex)
    )) {isSetParameters=true}
   /SetVariables(Emitter.Des:NiagaraFloat=&self:ScratchDynamicInput_08(
     Blot:NiagaraFloat=0.3
     Citadel:NiagaraFloat=0.3
     Froth:NiagaraFloat=1
     VariantColorIndex:NiagaraInt32=&MakeIntFromFloat(Float=$User.VariantColorIndex)
    )) {isSetParameters=true}
   &EmitterState()
   &SpawnRate(SpawnRate=35, Use Spawn Probability=false)
  ]
  particleSpawnScript [
   &InitializeParticle(
    Color=<rgba 5 5 5 1>
    Lifetime Max=0.6
    Lifetime Min=0.4
    Lifetime Mode=::NewEnumerator1
    Sprite Rotation Mode=::NewEnumerator2
    Sprite Size Max=<20 30>
    Sprite Size Min=<10 15>
    Sprite Size Mode=::NewEnumerator2
    UsePositionOffset=false
    Write Material Random=true
   )
   &SocketLocation(SamplePosOffset=false, Socket Index=4, Socket Reader=$System.Socket Reader, Socket Sampling Mode=::NewEnumerator2, VelocitySpeedLimit=false)
   &SubUVAnimation(
    Evaluation Type=::NewEnumerator0
    "Sprite Renderer "=DataInterface() {propertyValues {SpriteRenderer=&self:Lightning_0.NiagaraSpriteRendererProperties_0}}
    SubUV Animation Mode:ENiagara_SubUVLookupModeV2=::NewEnumerator2
    UseEndFrame=false
    UseStartFrame=false
   )
   &AddVelocity()
  ]
  particleUpdateScript [
   &ParticleState()
   &ScaleSpriteSize(Uniform Curve Sprite Scale=DataInterface() {propertyValues {Curve {
       defaultValue=0
       keys [{interpMode=RCIM_Cubic, tangentMode=RCTM_User, time=0, value=1.8}, {interpMode=RCIM_Cubic, tangentMode=RCTM_User, time=1, value=1}]
      }}})
   &SpriteRotationRate(Rotation Rate=&RandomRangeFloat(Minimum=-150, Maximum=150))
   &SolveForcesAndVelocity(Clamp Velocity=false, Limit Acceleration=false)
   &FN_Mesh_Attributes(
    Color=&MakeLinearColorFromVectorAndFloat(
     "Vector (RGB)"=&Multiply_Vector(
      A=&MakeVectorFromLinearColorRGB(Color=&self:ScratchDynamicInput_05(
        BlotColor:LinearColor=<rgba 0.3 0.0044 0 0.4>
        CitadelColor:LinearColor=<rgba 0 0.3333 1 0.4>
        FrothColor:LinearColor=<rgba 0 0 0 1>
        VariantColorIndex:NiagaraInt32=&MakeIntFromFloat(Float=$User.VariantColorIndex)
       ))
      B=&Multiply_VectorByFloat(Vector=<1 1 1>, Float=2)
     )
     "Float (Alpha)"=&Multiply_Float(
      A=&Multiply_Float(A=2, B=&FN_PickFloatDependOnPlatform(TargetPlatformActive=1, TargetPlatformNoActive=0.6))
      B=&self:ScratchDynamicInput_08(
       Blot:NiagaraFloat=1
       Citadel:NiagaraFloat=1
       Froth:NiagaraFloat=3
       VariantColorIndex:NiagaraInt32=&MakeIntFromFloat(Float=$User.VariantColorIndex)
      )
     )
    )
    Create Color Attribute=true
    Module Output=::NewEnumerator1
   )
   &DynamicMaterialParameters(
    Index 0 Param 4=&FloatFromCurve(FloatCurve=DataInterface() {propertyValues {Curve {defaultValue=1, keys [{tangentMode=RCTM_User, time=0, value=0.3}, {tangentMode=RCTM_User, time=1, value=1}]}}})
    Module.Index0Float0Write=true
    Module.Index0Float1Write=true
    Module.Index0Float2Write=true
    Module.Index0Float3Write=true
   )
   &ScaleColor(
    Scale RGB=&Multiply_VectorByFloat(
     Vector=<1 1 1>
     Float=&FloatFromCurve(
      FloatCurve=DataInterface() {propertyValues {Curve {keys [
          <at 0 20 RCIM_Cubic>
          {arriveTangent=-36.5203, interpMode=RCIM_Cubic, leaveTangent=-36.5203, time=0.2181, value=10.7468}
          {arriveTangent=-11.1862, interpMode=RCIM_Cubic, leaveTangent=-11.1862, time=0.4873, value=2.204}
          <at 1 2>
         ]}}}
      Scale Curve=&FN_FrontEndFloatModifier(Gameplay=1, Front End=1)
     )
    )
    ScaleA=true
    ScaleRGB=true
   )
  ]
  rendererClasses [&NiagaraSpriteRendererProperties]
  renderers [
   {rendererClass=&NiagaraSpriteRendererProperties}
   {rendererClass=&NiagaraSpriteRendererProperties, rendererIndex=1}
  ]
 }
 TrailHighLight {
  emitterUpdateScript [
   &EmitterState()
   &FN_SpawnAnimTrail()
  ]
  particleSpawnScript [
   &FN_InitializeAnimTrailBoneIndecies()
   &FN_InitializeAnimTrail(
    AnimTrailWidth=1.5
    Color=&self:ScratchDynamicInput_05(
     BlotColor:LinearColor=<rgba 25 2.4219 2.7566 1>
     CitadelColor:LinearColor=<rgba 0 0.3333 1 1>
     FrothColor:LinearColor=<rgba 0 0 0 1>
     VariantColorIndex:NiagaraInt32=&MakeIntFromFloat(Float=$User.VariantColorIndex)
    )
    Lifetime=0.2
   )
  ]
  particleUpdateScript [
   &ParticleState()
   &FN_UpdateAnimTrail()
   &ScaleRibbonWidth(Ribbon Width Scale=&FloatFromCurve(FloatCurve=DataInterface() {propertyValues {Curve {defaultValue=1.2, keys [{tangentMode=RCTM_User, time=0, value=1}, {tangentMode=RCTM_User, time=1, value=0}]}}}))
   &SolveForcesAndVelocity(Clamp Velocity=false, Limit Acceleration=false)
   &ScaleColor(
    Scale RGB=&Multiply_VectorByFloat(
     Vector=<1 1 1>
     Float=&FloatFromCurve(FloatCurve=DataInterface() {propertyValues {Curve {keys [{arriveTangent=-1.7344, interpMode=RCIM_Cubic, leaveTangent=-1.7344, tangentMode=RCTM_User, time=0, value=5}, <at 0.94 0>]}}})
    )
    ScaleA=true
    ScaleRGB=true
   )
  ]
  rendererClasses [&NiagaraRibbonRendererProperties]
  renderers [{rendererClass=&NiagaraRibbonRendererProperties}]
 }
 TrailMain {
  emitterUpdateScript [
   /SetVariables(Emitter.Des:NiagaraFloat=&self:ScratchDynamicInput_08(
     Blot:NiagaraFloat=0.3
     Citadel:NiagaraFloat=0.3
     Froth:NiagaraFloat=1
     VariantColorIndex:NiagaraInt32=&MakeIntFromFloat(Float=$User.VariantColorIndex)
    )) {isSetParameters=true}
   &EmitterState()
   &FN_SpawnAnimTrail()
  ]
  particleSpawnScript [
   &FN_InitializeAnimTrailBoneIndecies()
   &FN_InitializeAnimTrail(AnimTrailWidth=2.2, Lifetime=0.6)
  ]
  particleUpdateScript [
   /SetVariables(Particles.BrightnessAll:NiagaraFloat=&FloatFromCurve(FloatCurve=DataInterface() {propertyValues {Curve {keys [<at 0 20>,<at 0.4 1>,<at 1 1>]}}})) {isSetParameters=true}
   &ParticleState()
   &FN_UpdateAnimTrail()
   &ScaleRibbonWidth(Ribbon Width Scale=&FloatFromCurve(FloatCurve=DataInterface() {propertyValues {Curve {
        defaultValue=0
        keys [
         {interpMode=RCIM_Cubic, tangentMode=RCTM_User, time=0, value=1.2}
         {arriveTangent=-0.766, interpMode=RCIM_Cubic, leaveTangent=-0.766, tangentMode=RCTM_User, time=0.7, value=0.7864}
         {arriveTangent=-1.7837, interpMode=RCIM_Cubic, leaveTangent=-1.7837, tangentMode=RCTM_User, time=1, value=0}
        ]
       }}}))
   &DynamicMaterialParameters(
    Index 0 Param 3=&FloatFromCurve(FloatCurve=DataInterface() {propertyValues {Curve {keys [<at 0 0.2>,<at 1 0.8>]}}})
    Module.Index0Float0Write=true
    Module.Index0Float1Write=true
    Module.Index0Float2Write=true
    Module.Index0Float3Write=true
   )
   &FN_Mesh_Attributes(
    Color=&MakeLinearColorFromVectorAndFloat(
     "Vector (RGB)"=&Multiply_Vector(
      A=&MakeVectorFromLinearColorRGB(Color=&self:ScratchDynamicInput_05(
        BlotColor:LinearColor=<rgba 1 0.0083 0 0.4>
        CitadelColor:LinearColor=<rgba 0 0.3 0.9 0.4>
        FrothColor:LinearColor=<rgba 0 0 0 1>
        VariantColorIndex:NiagaraInt32=&MakeIntFromFloat(Float=$User.VariantColorIndex)
       ))
      B=&Multiply_VectorByFloat(Vector=<1 1 1>, Float=3)
     )
     "Float (Alpha)"=&Multiply_Float(
      A=0.25
      B=&FloatFromCurve(
       FloatCurve=DataInterface() {propertyValues {Curve {keys [<at 0 5>,<at 0.2 1>,<at 1 1>]}}}
       Scale Curve=&self:ScratchDynamicInput_08(Blot:NiagaraFloat=1, Citadel:NiagaraFloat=1, Froth:NiagaraFloat=5, VariantColorIndex:NiagaraInt32=0)
      )
     )
    )
    Create Color Attribute=true
    Module Output=::NewEnumerator1
   )
   &SolveForcesAndVelocity(Clamp Velocity=false, Limit Acceleration=false)
   &ScaleColor(
    Scale RGB=&Multiply_VectorByFloat(
     Vector=&Multiply_VectorByFloat(
      Vector=<1 1 1>
      Float=&FN_PickFloatDependOnPlatform(TargetPlatformActive=1, TargetPlatformNoActive=1)
     )
     Float=$Particles.BrightnessAll
    )
    ScaleA=true
    ScaleRGB=true
   )
  ]
  rendererClasses [&NiagaraRibbonRendererProperties]
  renderers [
   {rendererClass=&NiagaraRibbonRendererProperties}
   {rendererClass=&NiagaraRibbonRendererProperties, rendererIndex=1}
  ]
 }
}
summary {
 systemName=NS_Pickaxe_SheerWill_Swing
 emitters [
  {emitterName=TrailMain, rendererClasses [&NiagaraRibbonRendererProperties]}
  {emitterName=Lightning, rendererClasses [&NiagaraSpriteRendererProperties]}
  {emitterName=TrailHighLight, rendererClasses [&NiagaraRibbonRendererProperties]}
  {emitterName=Dragonsparks, rendererClasses [&NiagaraSpriteRendererProperties]}
 ]
 userVariables [
  User.bIsFrontEnd {type=NiagaraBool}
  User.bIsFrontEndPreview {type=NiagaraBool}
  User.RibbonIDOffset {type=NiagaraInt32}
  User.VariantColorIndex {type=NiagaraFloat, defaultValue=0, description=""}
 ]
}
