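# Anim Blueprint for the Sheer Will pickaxe that animates its shard bones.
# The cached input pose feeds two Modify Bone chains, which share fixed base-shard offsets, into a Blend Poses by bool.
# Only the second chain first runs spring-limited, root-relative AnimDynamics on eleven dyn_shards bones.
# The update event computes six time-staggered sine bob vectors, but the AnimGraph shows zero dyn-shard offsets and no reads of those vectors.
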AnimBlueprint /BRCosmetics/Weapons/FORT_Melee/Pickaxe_SheerWill/Meshes/SheerWill_Axe_AnimBP
target_skeleton=&/SheerWill_Axe_Skeleton
graphs [
 AnimGraph {nodes [
   n1 = "Transform (Modify) Bone - Bone: dyn_shards_1"(
    ComponentPose=Local To Component(LocalPose=Use cached pose 'cache_Input_pose'())
    Alpha=1
    Translation="0,0,0"
    Rotation="0,0,0"
    Scale="1,1,1"
   ) {
    rotationMode=BMM_Additive
    scaleSpace=BCS_BoneSpace
    translationMode=BMM_Additive
    translationSpace=BCS_BoneSpace
    alphaBoolBlend {blendOption=Linear}
    boneToModify {boneName=dyn_shards_1}
   }
   n2 = "Transform (Modify) Bone - Bone: dyn_shards_2"(ComponentPose=@n1, Alpha=1, Translation="0,0,0", Rotation="0,0,0", Scale="1,1,1") {
    rotationMode=BMM_Additive
    scaleSpace=BCS_BoneSpace
    translationMode=BMM_Additive
    translationSpace=BCS_BoneSpace
    alphaBoolBlend {blendOption=Linear}
    boneToModify {boneName=dyn_shards_2}
   }
   n3 = "Transform (Modify) Bone - Bone: dyn_shards_3"(ComponentPose=@n2, Alpha=1, Translation="0,0,0", Rotation="0,0,0", Scale="1,1,1") {
    rotationMode=BMM_Additive
    scaleSpace=BCS_BoneSpace
    translationMode=BMM_Additive
    translationSpace=BCS_BoneSpace
    alphaBoolBlend {blendOption=Linear}
    boneToModify {boneName=dyn_shards_3}
   }
   n4 = "Transform (Modify) Bone - Bone: dyn_shards_4"(ComponentPose=@n3, Alpha=1, Translation="0,0,0", Rotation="0,0,0", Scale="1,1,1") {
    rotationMode=BMM_Additive
    scaleSpace=BCS_BoneSpace
    translationMode=BMM_Additive
    translationSpace=BCS_BoneSpace
    alphaBoolBlend {blendOption=Linear}
    boneToModify {boneName=dyn_shards_4}
   }
   n5 = "Transform (Modify) Bone - Bone: dyn_shards_5"(ComponentPose=@n4, Alpha=1, Translation="0,0,0", Rotation="0,0,0", Scale="1,1,1") {
    rotationMode=BMM_Additive
    scaleSpace=BCS_BoneSpace
    translationMode=BMM_Additive
    translationSpace=BCS_BoneSpace
    alphaBoolBlend {blendOption=Linear}
    boneToModify {boneName=dyn_shards_5}
   }
   n6 = "Transform (Modify) Bone - Bone: dyn_shards_6"(ComponentPose=@n5, Alpha=1, Translation="0,0,0", Rotation="0,0,0", Scale="1,1,1") {
    rotationMode=BMM_Additive
    scaleSpace=BCS_BoneSpace
    translationMode=BMM_Additive
    translationSpace=BCS_BoneSpace
    alphaBoolBlend {blendOption=Linear}
    boneToModify {boneName=dyn_shards_6}
   }
   n7 = "Transform (Modify) Bone - Bone: dyn_shards_7"(ComponentPose=@n6, Alpha=1, Translation="0,0,0", Rotation="0,0,0", Scale="1,1,1") {
    rotationMode=BMM_Additive
    scaleSpace=BCS_BoneSpace
    translationMode=BMM_Additive
    translationSpace=BCS_BoneSpace
    alphaBoolBlend {blendOption=Linear}
    boneToModify {boneName=dyn_shards_7}
   }
   n8 = "Transform (Modify) Bone - Bone: dyn_shards_8"(ComponentPose=@n7, Alpha=1, Translation="0,0,0", Rotation="0,0,0", Scale="1,1,1") {
    rotationMode=BMM_Additive
    scaleSpace=BCS_BoneSpace
    translationMode=BMM_Additive
    translationSpace=BCS_BoneSpace
    alphaBoolBlend {blendOption=Linear}
    boneToModify {boneName=dyn_shards_8}
   }
   n9 = "Transform (Modify) Bone - Bone: dyn_shards_9"(ComponentPose=@n8, Alpha=1, Translation="0,0,0", Rotation="0,0,0", Scale="1,1,1") {
    rotationMode=BMM_Additive
    scaleSpace=BCS_BoneSpace
    translationMode=BMM_Additive
    translationSpace=BCS_BoneSpace
    alphaBoolBlend {blendOption=Linear}
    boneToModify {boneName=dyn_shards_9}
   }
   n10 = "Transform (Modify) Bone - Bone: dyn_shards_10"(ComponentPose=@n9, Alpha=1, Translation="0,0,0", Rotation="0,0,0", Scale="1,1,1") {
    rotationMode=BMM_Additive
    scaleSpace=BCS_BoneSpace
    translationMode=BMM_Additive
    translationSpace=BCS_BoneSpace
    alphaBoolBlend {blendOption=Linear}
    boneToModify {boneName=dyn_shards_10}
   }
   n11 = "Transform (Modify) Bone - Bone: dyn_shards_11"(ComponentPose=@n10, Alpha=1, Translation="0,0,0", Rotation="0,0,0", Scale="1,1,1") {
    rotationMode=BMM_Additive
    scaleSpace=BCS_BoneSpace
    translationMode=BMM_Additive
    translationSpace=BCS_BoneSpace
    alphaBoolBlend {blendOption=Linear}
    boneToModify {boneName=dyn_shards_11}
   }
   n12 = "Transform (Modify) Bone - Bone: base_shards_1"(ComponentPose=@n11, Alpha=1, Translation="0,0,2", Rotation="0,5,0", Scale="1,1,1") {
    rotationMode=BMM_Additive
    scaleSpace=BCS_BoneSpace
    translationMode=BMM_Additive
    translationSpace=BCS_BoneSpace
    alphaBoolBlend {blendOption=Linear}
    boneToModify {boneName=base_shards_1}
    rotation {yaw=5}
    translation {z=2}
   }
   n13 = "Transform (Modify) Bone - Bone: base_shards_7"(ComponentPose=@n12, Alpha=1, Translation="0,0,3", Rotation="0,0,0", Scale="1,1,1") {
    rotationMode=BMM_Additive
    scaleSpace=BCS_BoneSpace
    translationMode=BMM_Additive
    translationSpace=BCS_BoneSpace
    alphaBoolBlend {blendOption=Linear}
    boneToModify {boneName=base_shards_7}
    translation {z=3}
   }
   n14 = "Transform (Modify) Bone - Bone: base_shards_10"(ComponentPose=@n13, Alpha=1, Translation="0,0,3", Rotation="0,0,0", Scale="1,1,1") {
    rotationMode=BMM_Additive
    scaleSpace=BCS_BoneSpace
    translationMode=BMM_Additive
    translationSpace=BCS_BoneSpace
    alphaBoolBlend {blendOption=Linear}
    boneToModify {boneName=base_shards_10}
    translation {z=3}
   }
   n15 = "Transform (Modify) Bone - Bone: base_shards_11"(ComponentPose=@n14, Alpha=1, Translation="0,0,3", Rotation="0,0,0", Scale="1,1,1") {
    rotationMode=BMM_Additive
    scaleSpace=BCS_BoneSpace
    translationMode=BMM_Additive
    translationSpace=BCS_BoneSpace
    alphaBoolBlend {blendOption=Linear}
    boneToModify {boneName=base_shards_11}
    translation {z=3}
   }
   n16 = Anim Dynamics - Bone: dyn_shards_2(ComponentPose=Local To Component(LocalPose=Use cached pose 'cache_Input_pose'()), Alpha=1, ExternalForce="0,0,0") {
    angularDampingOverride=1
    angularSpringConstant=20000
    bAngularSpring=true
    bOverrideAngularDamping=true
    bOverrideLinearDamping=true
    bUseGravityOverride=true
    gravityScale=0.35
    linearDampingOverride=1
    simulationSpace=RootRelative
    boundBone {boneName=dyn_shards_2}
    componentLinearAccScale=<1 1 1>
    componentLinearVelScale=<1 1 1>
    physicsBodyDefinitions [{
      boundBone {boneName=dyn_shards_2}
      boxExtents=<20 30 30>
      constraintSetup {angularLimitsMax=<5 5 5>, angularLimitsMin=<-5 -5 -5>, angularTarget=<1 0 0>, linearAxesMax=<1 1 1>, linearAxesMin=<-1 -1 -1>}
      localJointOffset=<2 0 0>
     }]
    relativeSpaceBone {boneName=b_AR_Root}
    simSpaceSettings {simSpaceAngularAlpha=1}
   }
   n17 = Anim Dynamics - Bone: dyn_shards_1(ComponentPose=@n16, Alpha=1, ExternalForce="0,0,0") {
    angularDampingOverride=1
    angularSpringConstant=20000
    bAngularSpring=true
    bOverrideAngularDamping=true
    bOverrideLinearDamping=true
    bUseGravityOverride=true
    gravityScale=0.35
    linearDampingOverride=1
    simulationSpace=RootRelative
    boundBone {boneName=dyn_shards_1}
    componentLinearAccScale=<1 1 1>
    componentLinearVelScale=<1 1 1>
    physicsBodyDefinitions [{
      boundBone {boneName=dyn_shards_1}
      boxExtents=<20 30 30>
      constraintSetup {angularLimitsMax=<5 5 5>, angularLimitsMin=<-5 -5 -5>, angularTarget=<1 0 0>, linearAxesMax=<1 2 1>, linearAxesMin=<-1 -1 -1>}
      localJointOffset=<1 0 0>
     }]
    relativeSpaceBone {boneName=b_AR_Root}
    simSpaceSettings {simSpaceAngularAlpha=1}
   }
   n18 = Anim Dynamics - Bone: dyn_shards_3(ComponentPose=@n17, Alpha=1, ExternalForce="0,0,0") {
    angularDampingOverride=1
    angularSpringConstant=20000
    bAngularSpring=true
    bOverrideAngularDamping=true
    bOverrideLinearDamping=true
    bUseGravityOverride=true
    gravityScale=0.35
    linearDampingOverride=1
    simulationSpace=RootRelative
    boundBone {boneName=dyn_shards_3}
    componentLinearAccScale=<1 1 1>
    componentLinearVelScale=<1 1 1>
    physicsBodyDefinitions [{
      boundBone {boneName=dyn_shards_3}
      boxExtents=<20 30 30>
      constraintSetup {angularLimitsMax=<5 5 5>, angularLimitsMin=<-5 -5 -5>, angularTarget=<1 0 0>, linearAxesMax=<1 1 1>, linearAxesMin=<-1 -1 -1>}
      localJointOffset=<4 0 0>
     }]
    relativeSpaceBone {boneName=b_AR_Root}
    simSpaceSettings {simSpaceAngularAlpha=1}
   }
   n19 = Anim Dynamics - Bone: dyn_shards_4(ComponentPose=@n18, Alpha=1, ExternalForce="0,0,0") {
    angularDampingOverride=1
    angularSpringConstant=20000
    bAngularSpring=true
    bOverrideAngularDamping=true
    bOverrideLinearDamping=true
    bUseGravityOverride=true
    gravityScale=0.35
    linearDampingOverride=1
    simulationSpace=RootRelative
    boundBone {boneName=dyn_shards_4}
    componentLinearAccScale=<1 1 1>
    componentLinearVelScale=<1 1 1>
    physicsBodyDefinitions [{
      boundBone {boneName=dyn_shards_4}
      boxExtents=<20 30 30>
      constraintSetup {angularLimitsMax=<5 5 5>, angularLimitsMin=<-5 -5 -5>, angularTarget=<1 0 0>, linearAxesMax=<1 1 1>, linearAxesMin=<-1 -1 -1>}
      localJointOffset=<2 0 0>
     }]
    relativeSpaceBone {boneName=b_AR_Root}
    simSpaceSettings {simSpaceAngularAlpha=1}
   }
   n20 = Anim Dynamics - Bone: dyn_shards_5(ComponentPose=@n19, Alpha=1, ExternalForce="0,0,0") {
    angularDampingOverride=1
    angularSpringConstant=20000
    bAngularSpring=true
    bOverrideAngularDamping=true
    bOverrideLinearDamping=true
    bUseGravityOverride=true
    gravityScale=0.35
    linearDampingOverride=1
    simulationSpace=RootRelative
    boundBone {boneName=dyn_shards_5}
    componentLinearAccScale=<1 1 1>
    componentLinearVelScale=<1 1 1>
    physicsBodyDefinitions [{
      boundBone {boneName=dyn_shards_5}
      boxExtents=<20 30 30>
      constraintSetup {angularLimitsMax=<5 5 5>, angularLimitsMin=<-5 -5 -5>, angularTarget=<1 0 0>, linearAxesMax=<1 1 1>, linearAxesMin=<-1 -1 -1>}
      localJointOffset=<4 0 0>
     }]
    relativeSpaceBone {boneName=b_AR_Root}
    simSpaceSettings {simSpaceAngularAlpha=1}
   }
   n21 = Anim Dynamics - Bone: dyn_shards_6(ComponentPose=@n20, Alpha=1, ExternalForce="0,0,0") {
    angularDampingOverride=1
    angularSpringConstant=20000
    bAngularSpring=true
    bOverrideAngularDamping=true
    bOverrideLinearDamping=true
    bUseGravityOverride=true
    gravityScale=0.35
    linearDampingOverride=1
    simulationSpace=RootRelative
    boundBone {boneName=dyn_shards_6}
    componentLinearAccScale=<1 1 1>
    componentLinearVelScale=<1 1 1>
    physicsBodyDefinitions [{
      boundBone {boneName=dyn_shards_6}
      boxExtents=<20 30 30>
      constraintSetup {angularLimitsMax=<5 5 5>, angularLimitsMin=<-5 -5 -5>, angularTarget=<1 0 0>, linearAxesMax=<1 1 1>, linearAxesMin=<-1 -1 -1>}
      localJointOffset=<4 0 0>
     }]
    relativeSpaceBone {boneName=b_AR_Root}
    simSpaceSettings {simSpaceAngularAlpha=1}
   }
   n22 = Anim Dynamics - Bone: dyn_shards_7(ComponentPose=@n21, Alpha=1, ExternalForce="0,0,0") {
    angularDampingOverride=1
    angularSpringConstant=20000
    bAngularSpring=true
    bOverrideAngularDamping=true
    bOverrideLinearDamping=true
    bUseGravityOverride=true
    gravityScale=0.35
    linearDampingOverride=1
    simulationSpace=RootRelative
    boundBone {boneName=dyn_shards_7}
    componentLinearAccScale=<1 1 1>
    componentLinearVelScale=<1 1 1>
    physicsBodyDefinitions [{
      boundBone {boneName=dyn_shards_7}
      boxExtents=<20 30 30>
      constraintSetup {angularLimitsMax=<5 5 5>, angularLimitsMin=<-5 -5 -5>, angularTarget=<1 0 0>, linearAxesMax=<1 1 1>, linearAxesMin=<-1 -1 -1>}
      localJointOffset=<2 0 0>
     }]
    relativeSpaceBone {boneName=b_AR_Root}
    simSpaceSettings {simSpaceAngularAlpha=1}
   }
   n23 = Anim Dynamics - Bone: dyn_shards_8(ComponentPose=@n22, Alpha=1, ExternalForce="0,0,0") {
    angularDampingOverride=1
    angularSpringConstant=20000
    bAngularSpring=true
    bOverrideAngularDamping=true
    bOverrideLinearDamping=true
    bUseGravityOverride=true
    gravityScale=0.35
    linearDampingOverride=1
    simulationSpace=RootRelative
    boundBone {boneName=dyn_shards_8}
    componentLinearAccScale=<1 1 1>
    componentLinearVelScale=<1 1 1>
    physicsBodyDefinitions [{
      boundBone {boneName=dyn_shards_8}
      boxExtents=<20 30 30>
      constraintSetup {angularLimitsMax=<5 5 5>, angularLimitsMin=<-5 -5 -5>, angularTarget=<1 0 0>, linearAxesMax=<1 1 1>, linearAxesMin=<-1 -1 -1>}
      localJointOffset=<2 0 0>
     }]
    relativeSpaceBone {boneName=b_AR_Root}
    simSpaceSettings {simSpaceAngularAlpha=1}
   }
   n24 = Anim Dynamics - Bone: dyn_shards_9(ComponentPose=@n23, Alpha=1, ExternalForce="0,0,0") {
    angularDampingOverride=1
    angularSpringConstant=20000
    bAngularSpring=true
    bOverrideAngularDamping=true
    bOverrideLinearDamping=true
    bUseGravityOverride=true
    gravityScale=0.35
    linearDampingOverride=1
    simulationSpace=RootRelative
    boundBone {boneName=dyn_shards_9}
    componentLinearAccScale=<1 1 1>
    componentLinearVelScale=<1 1 1>
    physicsBodyDefinitions [{
      boundBone {boneName=dyn_shards_9}
      boxExtents=<20 30 30>
      constraintSetup {angularLimitsMax=<5 5 5>, angularLimitsMin=<-5 -5 -5>, angularTarget=<1 0 0>, linearAxesMax=<1 1 1>, linearAxesMin=<-1 -1 -1>}
      localJointOffset=<4 0 0>
     }]
    relativeSpaceBone {boneName=b_AR_Root}
    simSpaceSettings {simSpaceAngularAlpha=1}
   }
   n25 = Anim Dynamics - Bone: dyn_shards_10(ComponentPose=@n24, Alpha=1, ExternalForce="0,0,0") {
    angularDampingOverride=1
    angularSpringConstant=20000
    bAngularSpring=true
    bOverrideAngularDamping=true
    bOverrideLinearDamping=true
    bUseGravityOverride=true
    gravityScale=0.35
    linearDampingOverride=1
    simulationSpace=RootRelative
    boundBone {boneName=dyn_shards_10}
    componentLinearAccScale=<1 1 1>
    componentLinearVelScale=<1 1 1>
    physicsBodyDefinitions [{
      boundBone {boneName=dyn_shards_10}
      boxExtents=<20 30 30>
      constraintSetup {angularLimitsMax=<5 5 5>, angularLimitsMin=<-5 -5 -5>, angularTarget=<1 0 0>, linearAxesMax=<1 1 1>, linearAxesMin=<-1 -1 -1>}
      localJointOffset=<2 0 0>
     }]
    relativeSpaceBone {boneName=b_AR_Root}
    simSpaceSettings {simSpaceAngularAlpha=1}
   }
   n26 = Anim Dynamics - Bone: dyn_shards_11(ComponentPose=@n25, Alpha=1, ExternalForce="0,0,0") {
    angularDampingOverride=1
    angularSpringConstant=20000
    bAngularSpring=true
    bOverrideAngularDamping=true
    bOverrideLinearDamping=true
    bUseGravityOverride=true
    gravityScale=0.35
    linearDampingOverride=1
    simulationSpace=RootRelative
    boundBone {boneName=dyn_shards_11}
    componentLinearAccScale=<1 1 1>
    componentLinearVelScale=<1 1 1>
    physicsBodyDefinitions [{
      boundBone {boneName=dyn_shards_11}
      boxExtents=<20 30 30>
      constraintSetup {angularLimitsMax=<5 5 5>, angularLimitsMin=<-5 -5 -5>, angularTarget=<1 0 0>, linearAxesMax=<1 1 1>, linearAxesMin=<-1 -1 -1>}
      localJointOffset=<4 0 0>
     }]
    relativeSpaceBone {boneName=b_AR_Root}
    simSpaceSettings {simSpaceAngularAlpha=1}
   }
   n27 = "Transform (Modify) Bone - Bone: dyn_shards_1"(ComponentPose=@n26, Alpha=1, Translation="0,0,0", Rotation="0,0,0", Scale="1,1,1") {
    rotationMode=BMM_Additive
    scaleSpace=BCS_BoneSpace
    translationMode=BMM_Additive
    translationSpace=BCS_BoneSpace
    alphaBoolBlend {blendOption=Linear}
    boneToModify {boneName=dyn_shards_1}
   }
   n28 = "Transform (Modify) Bone - Bone: dyn_shards_2"(ComponentPose=@n27, Alpha=1, Translation="0,0,0", Rotation="0,0,0", Scale="1,1,1") {
    rotationMode=BMM_Additive
    scaleSpace=BCS_BoneSpace
    translationMode=BMM_Additive
    translationSpace=BCS_BoneSpace
    alphaBoolBlend {blendOption=Linear}
    boneToModify {boneName=dyn_shards_2}
   }
   n29 = "Transform (Modify) Bone - Bone: dyn_shards_3"(ComponentPose=@n28, Alpha=1, Translation="0,0,0", Rotation="0,0,0", Scale="1,1,1") {
    rotationMode=BMM_Additive
    scaleSpace=BCS_BoneSpace
    translationMode=BMM_Additive
    translationSpace=BCS_BoneSpace
    alphaBoolBlend {blendOption=Linear}
    boneToModify {boneName=dyn_shards_3}
   }
   n30 = "Transform (Modify) Bone - Bone: dyn_shards_4"(ComponentPose=@n29, Alpha=1, Translation="0,0,0", Rotation="0,0,0", Scale="1,1,1") {
    rotationMode=BMM_Additive
    scaleSpace=BCS_BoneSpace
    translationMode=BMM_Additive
    translationSpace=BCS_BoneSpace
    alphaBoolBlend {blendOption=Linear}
    boneToModify {boneName=dyn_shards_4}
   }
   n31 = "Transform (Modify) Bone - Bone: dyn_shards_5"(ComponentPose=@n30, Alpha=1, Translation="0,0,0", Rotation="0,0,0", Scale="1,1,1") {
    rotationMode=BMM_Additive
    scaleSpace=BCS_BoneSpace
    translationMode=BMM_Additive
    translationSpace=BCS_BoneSpace
    alphaBoolBlend {blendOption=Linear}
    boneToModify {boneName=dyn_shards_5}
   }
   n32 = "Transform (Modify) Bone - Bone: dyn_shards_6"(ComponentPose=@n31, Alpha=1, Translation="0,0,0", Rotation="0,0,0", Scale="1,1,1") {
    rotationMode=BMM_Additive
    scaleSpace=BCS_BoneSpace
    translationMode=BMM_Additive
    translationSpace=BCS_BoneSpace
    alphaBoolBlend {blendOption=Linear}
    boneToModify {boneName=dyn_shards_6}
   }
   n33 = "Transform (Modify) Bone - Bone: dyn_shards_7"(ComponentPose=@n32, Alpha=1, Translation="0,0,0", Rotation="0,0,0", Scale="1,1,1") {
    rotationMode=BMM_Additive
    scaleSpace=BCS_BoneSpace
    translationMode=BMM_Additive
    translationSpace=BCS_BoneSpace
    alphaBoolBlend {blendOption=Linear}
    boneToModify {boneName=dyn_shards_7}
   }
   n34 = "Transform (Modify) Bone - Bone: dyn_shards_8"(ComponentPose=@n33, Alpha=1, Translation="0,0,0", Rotation="0,0,0", Scale="1,1,1") {
    rotationMode=BMM_Additive
    scaleSpace=BCS_BoneSpace
    translationMode=BMM_Additive
    translationSpace=BCS_BoneSpace
    alphaBoolBlend {blendOption=Linear}
    boneToModify {boneName=dyn_shards_8}
   }
   n35 = "Transform (Modify) Bone - Bone: dyn_shards_9"(ComponentPose=@n34, Alpha=1, Translation="0,0,0", Rotation="0,0,0", Scale="1,1,1") {
    rotationMode=BMM_Additive
    scaleSpace=BCS_BoneSpace
    translationMode=BMM_Additive
    translationSpace=BCS_BoneSpace
    alphaBoolBlend {blendOption=Linear}
    boneToModify {boneName=dyn_shards_9}
   }
   n36 = "Transform (Modify) Bone - Bone: dyn_shards_10"(ComponentPose=@n35, Alpha=1, Translation="0,0,0", Rotation="0,0,0", Scale="1,1,1") {
    rotationMode=BMM_Additive
    scaleSpace=BCS_BoneSpace
    translationMode=BMM_Additive
    translationSpace=BCS_BoneSpace
    alphaBoolBlend {blendOption=Linear}
    boneToModify {boneName=dyn_shards_10}
   }
   n37 = "Transform (Modify) Bone - Bone: dyn_shards_11"(ComponentPose=@n36, Alpha=1, Translation="0,0,0", Rotation="0,0,0", Scale="1,1,1") {
    rotationMode=BMM_Additive
    scaleSpace=BCS_BoneSpace
    translationMode=BMM_Additive
    translationSpace=BCS_BoneSpace
    alphaBoolBlend {blendOption=Linear}
    boneToModify {boneName=dyn_shards_11}
   }
   n38 = "Transform (Modify) Bone - Bone: base_shards_1"(ComponentPose=@n37, Alpha=1, Translation="0,0,2", Rotation="0,5,0", Scale="1,1,1") {
    rotationMode=BMM_Additive
    scaleSpace=BCS_BoneSpace
    translationMode=BMM_Additive
    translationSpace=BCS_BoneSpace
    alphaBoolBlend {blendOption=Linear}
    boneToModify {boneName=base_shards_1}
    rotation {yaw=5}
    translation {z=2}
   }
   n39 = "Transform (Modify) Bone - Bone: base_shards_7"(ComponentPose=@n38, Alpha=1, Translation="0,0,3", Rotation="0,0,0", Scale="1,1,1") {
    rotationMode=BMM_Additive
    scaleSpace=BCS_BoneSpace
    translationMode=BMM_Additive
    translationSpace=BCS_BoneSpace
    alphaBoolBlend {blendOption=Linear}
    boneToModify {boneName=base_shards_7}
    translation {z=3}
   }
   n40 = "Transform (Modify) Bone - Bone: base_shards_10"(ComponentPose=@n39, Alpha=1, Translation="0,0,3", Rotation="0,0,0", Scale="1,1,1") {
    rotationMode=BMM_Additive
    scaleSpace=BCS_BoneSpace
    translationMode=BMM_Additive
    translationSpace=BCS_BoneSpace
    alphaBoolBlend {blendOption=Linear}
    boneToModify {boneName=base_shards_10}
    translation {z=3}
   }
   n41 = "Transform (Modify) Bone - Bone: base_shards_11"(ComponentPose=@n40, Alpha=1, Translation="0,0,3", Rotation="0,0,0", Scale="1,1,1") {
    rotationMode=BMM_Additive
    scaleSpace=BCS_BoneSpace
    translationMode=BMM_Additive
    translationSpace=BCS_BoneSpace
    alphaBoolBlend {blendOption=Linear}
    boneToModify {boneName=base_shards_11}
    translation {z=3}
   }
   Output Pose(Result=Blend Poses by bool(BlendPose_0=Component To Local(ComponentPose=@n15), BlendPose_1=Component To Local(ComponentPose=@n41))) {layerGroup=None}
   AnimGraphNode_SubInput_0 = Input Pose()
   Save cached pose 'cache_Input_pose'(Pose=@AnimGraphNode_SubInput_0) {cacheName=cache_Input_pose}
  ]}
 EventGraph {nodes [
   Event BlueprintUpdateAnimation()
   > n1 = Set Range(Range:double=0.7) -> (Output_Get:double)
   > Set alpha(alpha:double=1) -> (Output_Get:double)
   n2 = float * float(A=float - float(A=GetRealTimeSeconds()), B=6)
   n3 = "Sin (Radians)"(A=@n2)
   > "Set Translate Shard 1 "("Translate Shard 1 ":Vector=MakeVector(X=float / float(A=@n3, B=@n1) {is=title K2Node_CallFunction}, Y=0)) -> (Output_Get:Vector)
   n4 = GetRealTimeSeconds()
   n5 = "Sin (Radians)"(A=float * float(A=float - float(A=@n4, B=0.2), B=7))
   > Set Translate Shard 2(Translate Shard 2:Vector=MakeVector(X=float / float(A=@n5, B=@n1), Y=0)) -> (Output_Get:Vector)
   n6 = GetRealTimeSeconds()
   n7 = "Sin (Radians)"(A=float * float(A=float - float(A=@n6, B=0.4), B=8))
   > Set Translate Shard 3(Translate Shard 3:Vector=MakeVector(X=float / float(A=@n7, B=@n1), Y=0)) -> (Output_Get:Vector)
   n8 = GetRealTimeSeconds()
   n9 = "Sin (Radians)"(A=float * float(A=float - float(A=@n8, B=0.6), B=7))
   > Set Translate Shard 4(Translate Shard 4:Vector=MakeVector(X=float / float(A=@n9, B=@n1), Y=0)) -> (Output_Get:Vector)
   n10 = GetRealTimeSeconds()
   n11 = "Sin (Radians)"(A=float * float(A=float - float(A=@n10, B=0.8), B=8))
   > Set Translate Shard 5(Translate Shard 5:Vector=MakeVector(X=float / float(A=@n11, B=@n1), Y=0)) -> (Output_Get:Vector)
   n12 = GetRealTimeSeconds()
   n13 = "Sin (Radians)"(A=float * float(A=float - float(A=@n12, B=1), B=7))
   > Set Translate Shard 6(Translate Shard 6:Vector=MakeVector(X=float / float(A=@n13, B=@n1), Y=0)) -> (Output_Get:Vector)
   TryGetPawnOwner()
   "Sin (Degrees)"!(A:double=@n2) -> (ReturnValue:double)
  ]}
]
variables [
 alpha {type=number}
 Range {type=number}
 "Translate Shard 1 " {type=Vector}
 Translate Shard 2 {type=Vector}
 Translate Shard 3 {type=Vector}
 Translate Shard 4 {type=Vector}
 Translate Shard 5 {type=Vector}
 Translate Shard 6 {type=Vector}
]
